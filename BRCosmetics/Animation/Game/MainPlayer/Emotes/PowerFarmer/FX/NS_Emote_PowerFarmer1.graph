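# Niagara effect for the PowerFarmer player emote that switches between land and water visuals.
# User.bIsInWater multiplies each emitter's spawn amount by zero or one, selecting the set.
# On land, sprite chunks, exhaust smoke and world smoke spawn per distance moved, with steady rates only in the front end.
# In water, mesh bubbles, two mesh splash emitters, sprite drops and a foam burst play, with drops and foam constrained to the water plane.
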NiagaraSystem /BRCosmetics/Animation/Game/MainPlayer/Emotes/PowerFarmer/FX/NS_Emote_PowerFarmer1
dependencies {
 usedDataInterfaces [
  &NiagaraDataInterfaceCurve
  &NiagaraDataInterfaceVectorCurve
  &NiagaraDataInterfaceVector2DCurve
  &NiagaraDataInterfaceColorCurve
 ]
 usedDynamicInputs []
 usedModules [/Niagara/Modules/System/SystemLifeCycle, &SystemState]
}
emitters {
 Bubbles {
  emitterUpdateScript [
   &EmitterState()
   &SpawnRate(
    SpawnRate=&Multiply_Float(A=&MakeCustomFloatFromBool(Bool=$User.bIsInWater, True Float=1, False Float=0), B=5)
    Use Spawn Probability=false
   ) {module=SpawnRate001}
  ]
  particleSpawnScript [
   /Niagara/Modules/Spawn/Initialization/InitializeParticle(
    Lifetime=&UniformRangedFloat(Minimum=0.4, Maximum=0.6)
    Module.Write Color=true
    Module.Write Lifetime=true
    Module.Write Mass=true
    Module.Write Material Random=true
    Module.Write Position=false
    Module.Write Scale=false
    Module.Write SpriteRotation=false
    Module.Write SpriteSize=false
    Position=&SimulationPosition()
   )
   /Niagara/Modules/Spawn/Orientation/V2/InitialMeshOrientation(
    Orientation Vector=/Niagara/DynamicInputs/Random/RandomVector(
     Vector Scale:NiagaraFloat=1
     Constraint Vector:Vector3f=<0 0 1>
     Randomness Mode:ENiagaraRandomnessMode=::NewEnumerator0
     Random Seed:NiagaraInt32=0
     Fixed Random Seed:NiagaraBool=false
    )
    Use Orientation Vector=true
    Use Rotation Vector=false
   )
   &AddVelocity(Velocity=<0 -150 25>)
   &InheritVelocity(Module.Limit Speed=true, Velocity Limit=1000, Velocity Scale=<0.5 0.5 0>)
   &SystemLocation(MaskSpawnGroup=false, Offset=<0 -50 0>) {enabled=false}
   &SystemLocation(
    MaskSpawnGroup=false
    Offset=&Lerp_Vector(A=<0 -50 0>, B=<0 50 0>, Alpha=&UniformAOrBFloat(A=0))
   ) {module=SystemLocation001}
  ]
  particleUpdateScript [
   &ParticleState()
   &GravityForce(Gravity=<0 0 50>)
   &SolveForcesAndVelocity(Clamp Velocity=false, Limit Acceleration=false)
   &ScaleColor(
    Scale Alpha=&FloatFromCurve(FloatCurve=DataInterface() {propertyValues {Curve {keys [<at 0 -1>,<at 0.05 0>,<at 0.7904 0.0132>,<at 1.0011 -0.9961>]}}})
    ScaleA=true
    ScaleRGB=true
   )
   &ScaleMeshSize(Scale Factor=&VectorFromFloat(Value=&FloatFromCurve(FloatCurve=DataInterface() {propertyValues {Curve {keys [
          {interpMode=RCIM_Cubic, tangentMode=RCTM_User, time=-0.0002, value=0.0422}
          {arriveTangent=0.8781, interpMode=RCIM_Cubic, leaveTangent=0.8781, tangentMode=RCTM_User, time=0.1611, value=0.6343}
          <at 0.9903 0.9119>
         ]}}})))
  ]
  rendererClasses [&NiagaraMeshRendererProperties]
  renderers [{rendererClass=&NiagaraMeshRendererProperties, rendererIndex=0}]
 }
 Chunks {
  emitterUpdateScript [
   &EmitterState(Enable Distance Culling=true, MaxDistance=750, Module.bUseMaxDistance=true, Module.bUseMinDistance=false, Scalability Mode=::NewEnumerator1)
   &SpawnRate(
    SpawnRate=&Multiply_Float(
     A=&MakeCustomFloatFromBool(Bool=$User.bIsInWater, True Float=0, False Float=1)
     B=&FN_FrontEndFloatModifier(Gameplay=0, Front End=5)
    )
    Use Spawn Probability=false
   )
   &SpawnPerUnit(
    Module.Use Max Movement Threshold=true
    Module.Use Movement Tolerance=true
    Movement Tolerance=1
    Spawn Per Unit:NiagaraFloat=&Multiply_Float(A=&MakeCustomFloatFromBool(Bool=$User.bIsInWater, True Float=0, False Float=1), B=25)
    Use Spawn Probability=false
   )
  ]
  particleSpawnScript [
   &InitializeParticle(
    Color=&ColorFromCurve(
     ColorCurve=DataInterface() {propertyValues {
       AlphaCurve {keys [<at -0.0023 1 RCIM_Cubic>]}
       BlueCurve {keys [<at -0.0023 0.2122 RCIM_Cubic>,<at 0.1383 0.2122>,<at 0.2178 0.1859>,<at 0.4486 0.75>,<at 0.7126 0.1859>,<at 1.0234 0.8802>]}
       GreenCurve {keys [<at -0.0023 0.2325 RCIM_Cubic>,<at 0.1383 0.2325>,<at 0.2178 0.2005>,<at 0.4486 0.8456>,<at 0.7126 0.2122>,<at 1.0234 0.9474>]}
       RedCurve {keys [<at -0.0023 0.25 RCIM_Cubic>,<at 0.1383 0.25>,<at 0.2178 0.2188>,<at 0.4486 1>,<at 0.7126 0.2188>,<at 1.0234 1>]}
      }}
     CurveIndex=&UniformRangedFloat(Minimum=0, Maximum=1)
    )
    Lifetime=&UniformRangedFloat(Minimum=0.25, Maximum=1)
    Module.Write Color=true
    Module.Write Lifetime=true
    Module.Write Mass=true
    Module.Write Material Random=true
    Module.Write Position=false
    Module.Write Scale=false
    Module.Write SpriteRotation=true
    Module.Write SpriteSize=true
    Position=&SimulationPosition()
    Sprite Rotation=&UniformRangedFloat(Minimum=0, Maximum=360)
    Sprite Size=&Vector2DFromFloat(Value=&UniformRangedFloat(Minimum=4, Maximum=10))
   )
   /Niagara/Modules/Spawn/Location/SphereLocation(MaskSpawnGroup=false, Module.Override Randomness=false, Non Uniform Scale=<0.25 0.25 0.25>, Sphere Radius=1, Surface Only=false)
   &AddVelocity(Velocity=<0 0 25>)
   &InheritVelocity(Module.Limit Speed=true, Velocity Scale=&UniformRangedVector(Minimum=<0 -5 0>, Maximum=<0 -5 0>))
   &AddVelocityInCone(
    Cone Angle=50
    Cone Axis=<0 -1 1>
    Module.Use Velocity Falloff On Cone Axis=false
    Velocity Strength=&UniformRangedFloat(Minimum=0, Maximum=250)
   )
   &SystemLocation(
    MaskSpawnGroup=false
    Offset=&Lerp_Vector(A=<0 -50 0>, B=<1 50 0>, Alpha=&UniformAOrBFloat(A=0))
   )
   &SphereLocation(MaskSpawnGroup=false, Module.Override Randomness=false, Sphere Radius=5, Surface Only=false) {module=SphereLocation001}
  ]
  particleUpdateScript [
   &ParticleState()
   &Drag(Use Linear Drag=true, Use Rotational Drag=false)
   &GravityForce(Gravity=<0 0 -500>)
   &SolveForcesAndVelocity(Clamp Velocity=false, Limit Acceleration=false)
   &ScaleColor(Scale RGB=<1.2 1.2 1.2>, ScaleA=true, ScaleRGB=true)
   &ScaleSpriteSize(Scale Factor=&Vector2DFromFloat(Value=&FloatFromCurve(FloatCurve=DataInterface() {propertyValues {Curve {keys [
          <at 0.7 0.9701>
          {arriveTangent=-3.2149, interpMode=RCIM_Cubic, leaveTangent=-3.2149, tangentMode=RCTM_User, time=0.9205, value=0.9191}
          <at 1.0011 0.002>
         ]}}})))
   &SpriteRotationRate()
   &ScaleSpriteSizeBySpeed(
    Max Scale Factor=&Vector2DFromFloat(Value=1)
    Min Scale Factor=&Vector2DFromFloat(Value=0.25)
    Velocity Threshold=200
   )
  ]
  rendererClasses [&NiagaraSpriteRendererProperties]
  renderers [{rendererClass=&NiagaraSpriteRendererProperties}]
 }
 FN_Water_Drops_ {
  emitterUpdateScript [
   &EmitterState()
   &SpawnRate(
    SpawnRate=&Multiply_Float(A=&MakeCustomFloatFromBool(Bool=$User.bIsInWater, True Float=1, False Float=0), B=25)
    Use Spawn Probability=false
   ) {module=SpawnRate001}
  ]
  particleSpawnScript [
   /SetVariables(Particles.Random=&UniformRangedFloat(Minimum=0, Maximum=1)) {isSetParameters=true}
   &InitializeParticle(
    Color=<rgba 5 5 5 1>
    Lifetime=&UniformRangedFloat(Minimum=0.5, Maximum=0.75)
    Mass=&UniformRangedFloat(Minimum=0.6, Maximum=1.2)
    Mesh Scale=<100 100 100>
    Module.Write Color=true
    Module.Write Lifetime=true
    Module.Write Mass=true
    Module.Write Material Random=true
    Module.Write Position=false
    Module.Write Scale=false
    Module.Write SpriteRotation=true
    Module.Write SpriteSize=true
    Position=&SimulationPosition()
    Sprite Size=&Vector2DFromCurve(
     Vector2Curve=DataInterface() {propertyValues {
       XCurve {keys [<at 0 1.0924 RCIM_Cubic>,<at 0.9989 0.3489 RCIM_Cubic>]}
       YCurve {keys [
         {arriveTangent=-6.848, interpMode=RCIM_Cubic, leaveTangent=-6.848, tangentMode=RCTM_User, time=0.0012, value=4.1914}
         {arriveTangent=-2.264, interpMode=RCIM_Cubic, leaveTangent=-2.264, time=0.2419, value=3.2838}
         <at 0.9989 1.9328 RCIM_Cubic>
        ]}
      }}
     CurveIndex=$Particles.Random
     ScaleCurve=<0.25 0.25>
    )
   )
   &DynamicMaterialParameters(
    Index 0 Param 1=&UniformRangedFloat(Minimum=4, Maximum=10)
    Module.Param0WriteEnabled=true
    Module.Param1WriteEnabled=false
    Module.Param2WriteEnabled=false
    Module.Param3WriteEnabled=false
   )
   &InitializeWaterSnapToOwner()
   &ConstrainPositionToPlane(Plane Normal:Vector3f=$Particles.WaterPlaneNormal, Plane Position:NiagaraPosition=$User.WaterPlaneLocation) {module=ConstrainPositionToPlane001}
   &CylinderLocation(
    Cylinder Height=1
    Cylinder Radius=25
    Lathe Profile=DataInterface() {propertyValues {Curve {keys []}}}
    MaskSpawnGroup=false
    Module.Enable Lathe Profile=false
    Module.Override Randomness=false
    Module.Surface Only=true
    Surface Only Band Thickness=2
   )
   /SetVariables(Particles.SpriteSize=&Multiply_Vector2(
     A=$Particles.SpriteSize
     B=&Vector2DFromFloat(Value=&FloatFromCurve(
       FloatCurve=DataInterface() {propertyValues {Curve {keys [
           {arriveTangent=1.6964, interpMode=RCIM_Cubic, time=-0.0012, value=-0.0071}
           {arriveTangent=13.7683, interpMode=RCIM_Cubic, leaveTangent=13.7683, time=0.3287, value=17.2263}
           {arriveTangent=-3.3438, interpMode=RCIM_Cubic, leaveTangent=-3.3438, time=0.7139, value=9.8392}
           {arriveTangent=1.8535, interpMode=RCIM_Cubic, leaveTangent=1.8535, tangentMode=RCTM_User, time=0.9945, value=15}
          ]}}}
       CurveIndex=$Emitter.Scalability.DistanceFraction
      ))
    )) {isSetParameters=true}
   &PointForce(
    Force Origin=&SimulationPosition()
    Force Strength=&UniformRangedFloat(Minimum=100, Maximum=100)
    Module.Use Falloff=false
    Module.UseLinearForce=false
   ) {module=PointForce001}
   &ApplyInitialForces()
  ]
  particleUpdateScript [
   &ParticleState()
   /Game/Effects/Niagara/Emitters/Water/FN_DistanceKillParticles(AgeReduiceAmount=0.01)
   &ScaleColor(
    Scale Alpha=&FloatFromCurve(FloatCurve=DataInterface() {propertyValues {Curve {keys [
         <at 0 0 RCIM_Cubic>
         {arriveTangent=0.002, interpMode=RCIM_Cubic, leaveTangent=0.002, time=0.1183, value=1.0113}
         <at 1.0011 0.002 RCIM_Cubic>
        ]}}})
    ScaleA=true
    ScaleRGB=true
   )
   /Niagara/Modules/Update/Forces/CurlNoiseForce(
    Noise Frequency=1
    Noise Strength=&FloatFromCurve(FloatCurve=DataInterface() {propertyValues {Curve {keys [
         {arriveTangent=-1159.6794, interpMode=RCIM_Cubic, leaveTangent=-1159.6792, tangentMode=RCTM_User, time=-0.0046, value=500}
         {arriveTangent=-447.4286, interpMode=RCIM_Cubic, leaveTangent=-447.4286, time=0.2223, value=189.6793}
         <at 1.0011 50 RCIM_Cubic>
        ]}}})
    Pan Noise=false
   )
   &Drag(Use Linear Drag=true, Use Rotational Drag=false)
   &GravityForce()
   &SolveForcesAndVelocity(Clamp Velocity=false, Limit Acceleration=false)
  ]
  rendererClasses [&NiagaraSpriteRendererProperties]
  renderers [{rendererClass=&NiagaraSpriteRendererProperties}]
 }
 FN_Water_Foam_SnapToSurface001 {
  emitterUpdateScript [
   &EmitterState() {module=EmitterState001}
   &EmitterState()
   &SpawnBurst_Instantaneous(
    Spawn Count=&Multiply_Int(A=&MakeCustomIntFromBool(Bool=$User.bIsInWater, True INT=1, False INT=0), B=1)
    Use Spawn Probability=false
   )
  ]
  particleSpawnScript [
   &InitializeParticle(
    Lifetime=&UniformRangedFloat(Minimum=1, Maximum=3)
    Mass=&UniformRangedFloat(Minimum=0.6, Maximum=1.2)
    Mesh Scale=&VectorFromFloat(Value=&Multiply_Float(A=40, B=$User.EffectScale))
    Module.Write Color=true
    Module.Write Lifetime=true
    Module.Write Mass=true
    Module.Write Material Random=true
    Module.Write Position=false
    Module.Write Scale=true
    Module.Write SpriteRotation=true
    Module.Write SpriteSize=false
    Position=&SimulationPosition()
    Sprite Rotation=&NormalizedAngleToDegrees(NormalizedAngle=&UniformRangedFloat(Minimum=0, Maximum=1))
   )
   &InitialMeshOrientation(
    Orientation Vector=<0 0 0>
    Rotation=&UniformRangedVector(Minimum=<0 0 0>, Maximum=<0 0 1>)
    Use Orientation Vector=true
    Use Rotation Vector=true
   )
   &DynamicMaterialParameters(
    Index 1 Param 3=200
    Index 3 Param 1=$User.WaterBodyIndex
    Index 3 Param 2=$User.WaterDepth
    Module.Param0WriteEnabled1=false
    Module.Param0WriteEnabled3=true
    Module.Param1WriteEnabled1=false
    Module.Param1WriteEnabled3=true
    Module.Param2WriteEnabled1=true
    Module.Param2WriteEnabled3=true
    Module.Param3WriteEnabled1=false
    Module.Param3WriteEnabled3=true
    Write Parameter Index 0=false
    Write Parameter Index 1=true
    Write Parameter Index 3=true
   )
   &InitializeWaterSnapToOwner()
   &InheritVelocity(Module.Limit Speed=true)
  ]
  particleUpdateScript [
   &ParticleState()
   &DynamicMaterialParameters(
    Index 0 Param 1=&FloatFromCurve(FloatCurve=DataInterface() {propertyValues {Curve {keys [<at -0.0046 1>,<at 1.0011 4>]}}})
    Index 0 Param 2=&FloatFromCurve(FloatCurve=DataInterface() {propertyValues {Curve {keys [{arriveTangent=0.6218, interpMode=RCIM_Cubic, leaveTangent=0.6218, tangentMode=RCTM_User, time=-0.0046, value=0.5}, <at 1.0011 1 RCIM_Cubic>]}}})
    Index 0 Param 3=&FloatFromCurve(FloatCurve=DataInterface() {propertyValues {Curve {keys [<at -0.0046 1>,<at 1.0011 0.5>]}}})
    Index 2 Param 1=&FloatFromCurve(FloatCurve=DataInterface() {propertyValues {Curve {keys [<at -0.0046 0.4>,<at 1.0011 1.5>]}}})
    Index 2 Param 2=&FloatFromCurve(FloatCurve=DataInterface() {propertyValues {Curve {keys [<at -0.0046 0.1>,<at 0.11 2.6>,<at 1.0011 0.3>]}}})
    Index 2 Param 3=1
    Index 2 Param 4=&FloatFromCurve(FloatCurve=DataInterface() {propertyValues {Curve {keys [
         {arriveTangent=-1.9991, interpMode=RCIM_Cubic, time=0, value=1}
         {arriveTangent=-2.7143, interpMode=RCIM_Cubic, leaveTangent=-2.7143, tangentMode=RCTM_User, time=0.2095, value=0.6901}
         <at 1 0 RCIM_Cubic>
        ]}}})
    Module.Param0WriteEnabled=true
    Module.Param0WriteEnabled2=true
    Module.Param1WriteEnabled=true
    Module.Param1WriteEnabled2=true
    Module.Param2WriteEnabled=true
    Module.Param2WriteEnabled2=true
    Module.Param3WriteEnabled=true
    Module.Param3WriteEnabled2=true
    Write Parameter Index 2=true
   ) {module=DynamicMaterialParameters001}
   &Drag(Drag=2, Use Linear Drag=true, Use Rotational Drag=false)
   &SolveForcesAndVelocity(Clamp Velocity=false, Limit Acceleration=false)
   &ScaleMeshSize(Scale Factor=&VectorFromFloat(Value=&FloatFromCurve(FloatCurve=DataInterface() {propertyValues {Curve {keys [
          {arriveTangent=26.4511, interpMode=RCIM_Cubic, leaveTangent=26.451, tangentMode=RCTM_User, time=-0.0046, value=0.6073}
          {arriveTangent=3.3733, interpMode=RCIM_Cubic, leaveTangent=3.3733, time=0.1268, value=2.5016}
          <at 1.0011 4 RCIM_Cubic>
         ]}}}))) {module=MeshSizeScale}
   &ConstrainPositionToPlane(
    Plane Normal:Vector3f=$Particles.WaterPlaneNormal
    Plane Position:NiagaraPosition=&Add_Vector(A=Linked() {linkedVariable {name=User.WaterPlaneLocation, type=NiagaraPosition}}, B=<0 0 0>)
   )
   &CameraOffset(Camera Offset Amount=10)
  ]
  rendererClasses [&NiagaraMeshRendererProperties]
  renderers [{rendererClass=&NiagaraMeshRendererProperties, rendererIndex=0}]
 }
 SmokeExhaust {
  emitterUpdateScript [
   &EmitterState()
   &SpawnRate(
    SpawnRate=&Multiply_Float(
     A=&MakeCustomFloatFromBool(Bool=$User.bIsInWater, True Float=0, False Float=1)
     B=&FN_FrontEndFloatModifier(Gameplay=0, Front End=25)
    )
    Use Spawn Probability=false
   )
   &SpawnPerUnit(
    Module.Use Max Movement Threshold=true
    Module.Use Movement Tolerance=true
    Movement Tolerance=1
    Spawn Per Unit:NiagaraFloat=&Multiply_Float(A=&MakeCustomFloatFromBool(Bool=$User.bIsInWater, True Float=0, False Float=1), B=5)
    Use Spawn Probability=false
   )
  ]
  particleSpawnScript [
   &InitializeParticle(
    Color=&UniformRangedLinearColor(Minimum=<rgba 0.1875 0.185 0.1642 0.15>, Maximum=<rgba 0.0781 0.0775 0.0721 0.25>)
    Lifetime=&UniformRangedFloat(Minimum=0.5, Maximum=0.6)
    Module.Write Color=true
    Module.Write Lifetime=true
    Module.Write Mass=true
    Module.Write Material Random=true
    Module.Write Position=false
    Module.Write Scale=false
    Module.Write SpriteRotation=true
    Module.Write SpriteSize=true
    Position=&SimulationPosition()
    Sprite Rotation=&UniformRangedFloat(Minimum=0, Maximum=360)
    Sprite Size=&Vector2DFromFloat(Value=&UniformRangedFloat(Minimum=10, Maximum=25))
   )
   &TorusLocation(
    Align Torus With Axis=false
    Large Radius=10
    MaskSpawnGroup=false
    Module.Override Randomness=false
    Surface Position=1
    Torus Mode=::NewEnumerator2
    Torus Origin:NiagaraPosition=&Add_Vector(
     A=Linked() {linkedVariable {name=Particles.Position, type=NiagaraPosition}}
     B=&Multiply_VectorByFloat(Vector=$Engine.Owner.SystemZAxis, Float=0)
    )
   )
   &AccelerationForce(Acceleration=&UniformRangedVector(Minimum=<0 0 50>, Maximum=<0 0 100>))
   &PointForce(
    Force Origin=&SimulationPosition()
    Force Strength=&UniformRangedFloat(Minimum=100, Maximum=200)
    Module.Use Falloff=false
    Module.UseLinearForce=false
   )
   &ApplyInitialForces()
   &AddVelocity(Velocity=<0 -200 0>)
   &InheritVelocity(Module.Limit Speed=true, Velocity Limit=1000)
   &SystemLocation(MaskSpawnGroup=false, Offset=<0 -42 -5>)
  ]
  particleUpdateScript [
   &ParticleState()
   &Drag(Drag=7, Use Linear Drag=true, Use Rotational Drag=false)
   &SolveForcesAndVelocity(Clamp Velocity=false, Limit Acceleration=false)
   &ScaleSpriteSize(Scale Factor=&Vector2DFromFloat(Value=&FloatFromCurve(FloatCurve=DataInterface() {propertyValues {Curve {keys [
          {arriveTangent=1.491, interpMode=RCIM_Cubic, leaveTangent=1.491, tangentMode=RCTM_User, time=-0.0046, value=1.1969}
          <at 1.0011 3.4809 RCIM_Cubic>
         ]}}})))
   /SetVariables(Particles.SubImageIndex=&FloatFromCurve(FloatCurve=DataInterface() {propertyValues {Curve {keys [
         {arriveTangent=29.8623, interpMode=RCIM_Cubic, leaveTangent=29.8623, tangentMode=RCTM_User, time=-0.0046, value=5.0556}
         <at 1.0011 22.1689 RCIM_Cubic>
        ]}}})) {isSetParameters=true, module=SetVariables_#1}
   &DynamicMaterialParameters(
    Index 0 Param 1=&FloatFromCurve(FloatCurve=DataInterface() {propertyValues {Curve {keys [{interpMode=RCIM_Cubic, tangentMode=RCTM_User, time=0, value=1}, {interpMode=RCIM_Cubic, tangentMode=RCTM_User, time=1, value=0}]}}})
    Index 0 Param 4=&FloatFromCurve(FloatCurve=DataInterface() {propertyValues {Curve {keys [{interpMode=RCIM_Cubic, tangentMode=RCTM_User, time=0, value=0}, {interpMode=RCIM_Cubic, tangentMode=RCTM_User, time=1, value=1}]}}})
    Module.Param0WriteEnabled=true
    Module.Param1WriteEnabled=false
    Module.Param2WriteEnabled=false
    Module.Param3WriteEnabled=true
   ) {module=DynamicMatParams}
   &CameraOffset(Camera Offset Amount=15)
   &ScaleColor(
    Scale Alpha=&FloatFromCurve(FloatCurve=DataInterface() {propertyValues {Curve {keys [
         {interpMode=RCIM_Cubic, tangentMode=RCTM_User, time=0.0008, value=0.0189}
         <at 0.1384 0.9931>
         {arriveTangent=-1.1488, interpMode=RCIM_Cubic, leaveTangent=-1.1488, time=0.5, value=1.0113}
         <at 1.0011 0.002>
        ]}}})
    ScaleA=true
    ScaleRGB=false
   )
  ]
  rendererClasses [&NiagaraSpriteRendererProperties]
  renderers [{rendererClass=&NiagaraSpriteRendererProperties}]
 }
 SmokeWorld {
  emitterUpdateScript [
   &EmitterState()
   &SpawnRate(
    SpawnRate=&Multiply_Float(
     A=&MakeCustomFloatFromBool(Bool=$User.bIsInWater, True Float=0, False Float=1)
     B=&FN_FrontEndFloatModifier(Gameplay=0, Front End=25)
    )
    Use Spawn Probability=false
   )
   &SpawnPerUnit(
    Module.Use Max Movement Threshold=true
    Module.Use Movement Tolerance=true
    Movement Tolerance=3
    Spawn Per Unit:NiagaraFloat=&Multiply_Float(A=&MakeCustomFloatFromBool(Bool=$User.bIsInWater, True Float=0, False Float=1), B=20)
    Use Spawn Probability=false
   )
  ]
  particleSpawnScript [
   &InitializeParticle(
    Color=&UniformRangedLinearColor(Minimum=<rgba 0.4662 0.478 0.4385 0.1>, Maximum=<rgba 0.781 0.7633 0.6955 0.2>)
    Lifetime=&UniformRangedFloat(Minimum=0.5, Maximum=0.6)
    Module.Write Color=true
    Module.Write Lifetime=true
    Module.Write Mass=true
    Module.Write Material Random=true
    Module.Write Position=false
    Module.Write Scale=false
    Module.Write SpriteRotation=true
    Module.Write SpriteSize=true
    Position=&SimulationPosition()
    Sprite Rotation=&UniformRangedFloat(Minimum=0, Maximum=360)
    Sprite Size=&Vector2DFromFloat(Value=&UniformRangedFloat(Minimum=10, Maximum=25))
   )
   &TorusLocation(
    Align Torus With Axis=false
    Large Radius=3
    MaskSpawnGroup=false
    Module.Override Randomness=false
    Surface Position=1
    Torus Mode=::NewEnumerator2
    Torus Origin:NiagaraPosition=&Add_Vector(
     A=Linked() {linkedVariable {name=Particles.Position, type=NiagaraPosition}}
     B=&Multiply_VectorByFloat(Vector=$Engine.Owner.SystemZAxis, Float=0)
    )
   )
   &AccelerationForce(Acceleration=&UniformRangedVector(Minimum=<0 0 25>, Maximum=<0 0 75>))
   &PointForce(
    Force Origin=&SimulationPosition()
    Force Strength=&UniformRangedFloat(Minimum=100, Maximum=200)
    Module.Use Falloff=false
    Module.UseLinearForce=false
   )
   &ApplyInitialForces()
   &AddVelocity(Velocity=<0 -300 0>)
   &InheritVelocity(Module.Limit Speed=true, Velocity Limit=1000)
   &SystemLocation(
    MaskSpawnGroup=false
    Offset=&Lerp_Vector(A=<3 20 0>, B=<3 50 0>, Alpha=&UniformAOrBFloat(A=0))
   ) {enabled=false, module=SystemLocation001}
  ]
  particleUpdateScript [
   &ParticleState()
   &Drag(Drag=7, Use Linear Drag=true, Use Rotational Drag=false)
   &SolveForcesAndVelocity(Clamp Velocity=false, Limit Acceleration=false)
   &ScaleSpriteSize(Scale Factor=&Vector2DFromFloat(Value=&FloatFromCurve(FloatCurve=DataInterface() {propertyValues {Curve {keys [
          {arriveTangent=1.491, interpMode=RCIM_Cubic, leaveTangent=1.491, tangentMode=RCTM_User, time=-0.0046, value=1.1969}
          <at 0.75 3.4809 RCIM_Cubic>
         ]}}})))
   /SetVariables(Particles.SubImageIndex=&FloatFromCurve(FloatCurve=DataInterface() {propertyValues {Curve {keys [
         {arriveTangent=29.8623, interpMode=RCIM_Cubic, leaveTangent=29.8623, tangentMode=RCTM_User, time=-0.0046, value=5.0556}
         <at 1.0011 22.1689 RCIM_Cubic>
        ]}}})) {isSetParameters=true, module=SetVariables_#1}
   &DynamicMaterialParameters(
    Index 0 Param 1=&FloatFromCurve(FloatCurve=DataInterface() {propertyValues {Curve {keys [
         {arriveTangent=-2.3572, interpMode=RCIM_Cubic, leaveTangent=-2.3572, tangentMode=RCTM_User, time=0, value=1}
         {interpMode=RCIM_Cubic, tangentMode=RCTM_User, time=1, value=0}
        ]}}})
    Index 0 Param 4=&FloatFromCurve(FloatCurve=DataInterface() {propertyValues {Curve {keys [{interpMode=RCIM_Cubic, tangentMode=RCTM_User, time=0, value=0}, {interpMode=RCIM_Cubic, tangentMode=RCTM_User, time=1, value=1}]}}})
    Module.Param0WriteEnabled=true
    Module.Param1WriteEnabled=false
    Module.Param2WriteEnabled=false
    Module.Param3WriteEnabled=true
   ) {module=DynamicMatParams}
   &CameraOffset(Camera Offset Amount=5)
   &ScaleColor(
    Scale Alpha=&FloatFromCurve(
     FloatCurve=DataInterface() {propertyValues {Curve {keys [
         {arriveTangent=9.0725, interpMode=RCIM_Cubic, leaveTangent=9.0725, tangentMode=RCTM_User, time=0, value=0}
         {arriveTangent=0.4456, interpMode=RCIM_Cubic, leaveTangent=0.4456, tangentMode=RCTM_User, time=0.1408, value=0.9847}
         {arriveTangent=-1.146, interpMode=RCIM_Cubic, leaveTangent=-1.146, time=0.6, value=1.025}
         {arriveTangent=-0.0405, interpMode=RCIM_Cubic, leaveTangent=-0.0405, tangentMode=RCTM_User, time=1, value=0}
        ]}}}
     Scale Curve=0.2
    )
    ScaleA=true
    ScaleRGB=false
   )
  ]
  rendererClasses [&NiagaraSpriteRendererProperties]
  renderers [{rendererClass=&NiagaraSpriteRendererProperties}]
 }
 Water_splash_arms {
  emitterUpdateScript [
   &EmitterState()
   &SpawnRate(
    SpawnRate=&Multiply_Float(A=&MakeCustomFloatFromBool(Bool=$User.bIsInWater, True Float=1, False Float=0), B=15)
    Use Spawn Probability=false
   ) {module=SpawnRate001}
  ]
  particleSpawnScript [
   &InitializeParticle(
    Lifetime=&Multiply_Float(A=&UniformRangedFloat(Minimum=0.5, Maximum=0.7), B=$User.EffectScale)
    Mass=&UniformRangedFloat(Minimum=0.6, Maximum=1.2)
    Mesh Scale=&Multiply_VectorByFloat(Vector=&UniformRangedVector(Minimum=<0.15 0.15 0.15>, Maximum=<0.2 0.2 0.2>), Float=1)
    Module.Write Color=true
    Module.Write Lifetime=true
    Module.Write Mass=true
    Module.Write Material Random=true
    Module.Write Position=false
    Module.Write Scale=true
    Module.Write SpriteRotation=true
    Module.Write SpriteSize=false
    Position=&SimulationPosition()
    Sprite Rotation=&NormalizedAngleToDegrees(NormalizedAngle=&UniformRangedFloat(Minimum=0, Maximum=1))
   )
   &TorusLocation(Align Torus With Axis=false, Large Radius=25, MaskSpawnGroup=false, Module.Override Randomness=false, Torus Mode=::NewEnumerator2)
   &ApplyInitialForces()
   &InitialMeshOrientation(
    Orientation Axis:Vector3f=<0 0 0>
    Orientation Vector=$Particles.Velocity
    Rotation=<0 0 -0.1>
    Use Orientation Vector=true
    Use Rotation Vector=true
   )
   &AddVelocity(Coordinate Space=::World)
   &SystemLocation(
    MaskSpawnGroup=false
    Offset=&Lerp_Vector(A=<0 35 0>, B=<0 -35 0>, Alpha=&UniformAOrBFloat(A=0))
   ) {module=SystemLocation001}
  ]
  particleUpdateScript [
   &ParticleState() {module=UpdateAge}
   &DynamicMaterialParameters(
    Index 0 Param 1=&FloatFromCurve(FloatCurve=DataInterface() {propertyValues {Curve {keys [
         {arriveTangent=0.8954, interpMode=RCIM_Cubic, leaveTangent=0.8954, tangentMode=RCTM_User, time=-0.0005, value=-1.845}
         {arriveTangent=1.8419, interpMode=RCIM_Cubic, leaveTangent=1.8419, time=0.7887, value=-0.7711}
         {arriveTangent=3.1029, interpMode=RCIM_Cubic, leaveTangent=3.1029, tangentMode=RCTM_User, time=1.0011, value=0}
        ]}}})
    Index 0 Param 2=&FloatFromCurve(FloatCurve=DataInterface() {propertyValues {Curve {keys [<at -0.0046 4>,<at 1.0011 0>]}}})
    Index 0 Param 3=&FloatFromCurve(
     FloatCurve=DataInterface() {propertyValues {Curve {keys [
         {arriveTangent=0.096, interpMode=RCIM_Cubic, leaveTangent=0.096, tangentMode=RCTM_User, time=-0.0046, value=0.0882}
         <at 1.0011 0.1372 RCIM_Cubic>
        ]}}}
     Scale Curve=&Multiply_Float(A=2, B=$User.EffectScale)
    )
    Index 0 Param 4=&FloatFromCurve(FloatCurve=DataInterface() {propertyValues {Curve {keys [
         {arriveTangent=1.9125, interpMode=RCIM_Cubic, leaveTangent=0.2728, tangentMode=RCTM_User, time=-0.0046, value=0.9276}
         {arriveTangent=4.6326, interpMode=RCIM_Cubic, leaveTangent=4.6326, tangentMode=RCTM_User, time=1.0011, value=2.7648}
        ]}}})
    Module.Param0WriteEnabled=true
    Module.Param1WriteEnabled=true
    Module.Param2WriteEnabled=true
    Module.Param3WriteEnabled=true
   ) {module=DynamicMaterialParameters001}
   &ScaleColor(
    Scale Alpha=&FloatFromCurve(FloatCurve=DataInterface() {propertyValues {Curve {keys [<at -0.0046 1.0113 RCIM_Cubic>,<at 1.0011 0.002 RCIM_Cubic>]}}})
    ScaleA=true
    ScaleRGB=false
   )
   &Drag(Drag=2, Use Linear Drag=true, Use Rotational Drag=false)
   &SolveForcesAndVelocity(Clamp Velocity=false, Limit Acceleration=false)
   &ScaleMeshSize(Scale Factor=&VectorFromCurve(VectorCurve=DataInterface() {propertyValues {
       XCurve {keys [<at 0.0044 1.4881>,<at 0.9989 3>]}
       YCurve {keys [<at -0.0012 1.4698>,<at 0.9989 3>]}
       ZCurve {keys [<at 0.0044 1.5612>,<at 0.9989 1.5622>]}
      }})) {module=MeshSizeScale}
   &CameraOffset(Camera Offset Amount=15)
  ]
  rendererClasses [&NiagaraMeshRendererProperties]
  renderers [{rendererClass=&NiagaraMeshRendererProperties, rendererIndex=0}]
 }
 Water_splash_arms001 {
  emitterUpdateScript [
   &EmitterState()
   &SpawnRate(
    SpawnRate=&Multiply_Float(A=&MakeCustomFloatFromBool(Bool=$User.bIsInWater, True Float=1, False Float=0), B=15)
    Use Spawn Probability=false
   ) {module=SpawnRate001}
  ]
  particleSpawnScript [
   &InitializeParticle(
    Lifetime=&Multiply_Float(A=&UniformRangedFloat(Minimum=0.5, Maximum=0.7), B=$User.EffectScale)
    Mass=&UniformRangedFloat(Minimum=0.6, Maximum=1.2)
    Mesh Scale=&UniformRangedVector(Minimum=<0.15 0.15 0.15>, Maximum=<0.2 0.2 0.2>)
    Module.Write Color=true
    Module.Write Lifetime=true
    Module.Write Mass=true
    Module.Write Material Random=true
    Module.Write Position=false
    Module.Write Scale=true
    Module.Write SpriteRotation=true
    Module.Write SpriteSize=false
    Position=&SimulationPosition()
    Sprite Rotation=&NormalizedAngleToDegrees(NormalizedAngle=&UniformRangedFloat(Minimum=0, Maximum=1))
   )
   &TorusLocation(Align Torus With Axis=false, Large Radius=25, MaskSpawnGroup=false, Module.Override Randomness=false, Torus Mode=::NewEnumerator2)
   &ApplyInitialForces()
   &InitialMeshOrientation(
    Orientation Axis:Vector3f=<0 0 0>
    Orientation Vector=$Particles.Velocity
    Rotation=<0 0 0.6>
    Use Orientation Vector=true
    Use Rotation Vector=true
   )
   &AddVelocity(Coordinate Space=::World)
   &SystemLocation(
    MaskSpawnGroup=false
    Offset=&Lerp_Vector(A=<0 35 0>, B=<0 -35 0>, Alpha=&UniformAOrBFloat(A=0))
   ) {module=SystemLocation001}
  ]
  particleUpdateScript [
   &ParticleState() {module=UpdateAge}
   &DynamicMaterialParameters(
    Index 0 Param 1=&FloatFromCurve(FloatCurve=DataInterface() {propertyValues {Curve {keys [
         {arriveTangent=0.8954, interpMode=RCIM_Cubic, leaveTangent=0.8954, tangentMode=RCTM_User, time=-0.0005, value=-1.845}
         {arriveTangent=1.8419, interpMode=RCIM_Cubic, leaveTangent=1.8419, time=0.7887, value=-0.7711}
         {arriveTangent=3.1029, interpMode=RCIM_Cubic, leaveTangent=3.1029, tangentMode=RCTM_User, time=1.0011, value=0}
        ]}}})
    Index 0 Param 2=&FloatFromCurve(FloatCurve=DataInterface() {propertyValues {Curve {keys [<at -0.0046 4>,<at 1.0011 0>]}}})
    Index 0 Param 3=&FloatFromCurve(
     FloatCurve=DataInterface() {propertyValues {Curve {keys [
         {arriveTangent=0.096, interpMode=RCIM_Cubic, leaveTangent=0.096, tangentMode=RCTM_User, time=-0.0046, value=0.0882}
         <at 1.0011 0.1372 RCIM_Cubic>
        ]}}}
     Scale Curve=&Multiply_Float(A=2, B=$User.EffectScale)
    )
    Index 0 Param 4=&FloatFromCurve(FloatCurve=DataInterface() {propertyValues {Curve {keys [
         {arriveTangent=1.9125, interpMode=RCIM_Cubic, leaveTangent=0.2728, tangentMode=RCTM_User, time=-0.0046, value=0.9276}
         {arriveTangent=4.6326, interpMode=RCIM_Cubic, leaveTangent=4.6326, tangentMode=RCTM_User, time=1.0011, value=2.7648}
        ]}}})
    Module.Param0WriteEnabled=true
    Module.Param1WriteEnabled=true
    Module.Param2WriteEnabled=true
    Module.Param3WriteEnabled=true
   ) {module=DynamicMaterialParameters001}
   &ScaleColor(
    Scale Alpha=&FloatFromCurve(FloatCurve=DataInterface() {propertyValues {Curve {keys [<at -0.0046 1.0113 RCIM_Cubic>,<at 1.0011 0.002 RCIM_Cubic>]}}})
    ScaleA=true
    ScaleRGB=false
   )
   &Drag(Drag=2, Use Linear Drag=true, Use Rotational Drag=false)
   &SolveForcesAndVelocity(Clamp Velocity=false, Limit Acceleration=false)
   &ScaleMeshSize(Scale Factor=&VectorFromCurve(VectorCurve=DataInterface() {propertyValues {
       XCurve {keys [<at 0.0044 1.4881>,<at 0.9989 3>]}
       YCurve {keys [<at -0.0012 1.4698>,<at 0.9989 3>]}
       ZCurve {keys [<at 0.0044 1.5612>,<at 0.9989 1.5622>]}
      }})) {module=MeshSizeScale}
   &CameraOffset(Camera Offset Amount=15)
  ]
  rendererClasses [&NiagaraMeshRendererProperties]
  renderers [{rendererClass=&NiagaraMeshRendererProperties, rendererIndex=0}]
 }
}
summary {
 systemName=NS_Emote_PowerFarmer1
 emitters [
  {emitterName=Water_splash_arms, rendererClasses [&NiagaraMeshRendererProperties]}
  {emitterName=FN_Water_Drops_, rendererClasses [&NiagaraSpriteRendererProperties]}
  {emitterName=SmokeWorld, rendererClasses [&NiagaraSpriteRendererProperties]}
  {emitterName=FN_Water_Foam_SnapToSurface001, rendererClasses [&NiagaraMeshRendererProperties]}
  {emitterName=SmokeExhaust, rendererClasses [&NiagaraSpriteRendererProperties]}
  {emitterName=Bubbles, rendererClasses [&NiagaraMeshRendererProperties]}
  {emitterName=Water_splash_arms001, rendererClasses [&NiagaraMeshRendererProperties]}
  {emitterName=Chunks, rendererClasses [&NiagaraSpriteRendererProperties]}
 ]
 userVariables [
  User.bIsFrontEnd {type=NiagaraBool}
  User.bIsFrontEndPreview {type=NiagaraBool}
  User.bIsInWater {type=NiagaraBool}
  User.EffectScale {type=NiagaraFloat, defaultValue=1, description=""}
  User.EffectSpawnScale {type=NiagaraFloat, defaultValue=10, description=""}
  User.ForwardSpeed {type=NiagaraFloat, defaultValue=0, description=""}
  User.FrontPontoonBig.Depth {type=NiagaraFloat, defaultValue=0, description=""}
  User.FrontPontoonBig.Impulse {type=Vector3f, description="", defaultValue {struct=Vector3f, value=<0 0 0>}}
  User.InWater {type=NiagaraBool}
  User.isBoosting {type=NiagaraBool}
  User.MidPontoonBig.Depth {type=NiagaraFloat, defaultValue=0, description=""}
  User.MidPontoonBig.Impulse {type=Vector3f, description="", defaultValue {struct=Vector3f, value=<0 0 0>}}
  User.NewVariable {type=NiagaraFloat, defaultValue=0, description=""}
  User.RearPontoonBig.Depth {type=NiagaraFloat, defaultValue=0, description=""}
  User.RearPontoonBig.Impulse {type=Vector3f, description="", defaultValue {struct=Vector3f, value=<0 0 0>}}
  User.WaterBodyIndex {type=NiagaraFloat, defaultValue=0, description=""}
  User.WaterDepth {type=NiagaraFloat, defaultValue=0, description=""}
  User.WaterPlaneLocation {type=NiagaraPosition, description="", defaultValue {}}
  User.WaterPlaneNormal {type=Vector3f, description="", defaultValue {struct=Vector3f, value=<0 0 0>}}
  User.WaveAttenuationFactor {type=NiagaraFloat, defaultValue=0, description=""}
  User.WaveTime {type=NiagaraFloat, defaultValue=0, description=""}
 ]
}
