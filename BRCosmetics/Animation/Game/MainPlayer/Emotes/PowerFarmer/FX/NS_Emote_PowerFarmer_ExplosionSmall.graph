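# Niagara system for a small explosion effect in the PowerFarmer player emote.
# Delayed one-shot emitters include two identical GPU dark-chunk sprite bursts with drag and curl noise, plus a brief orange particle light.
# Several emitters are present but disabled, including the stylized blast, flashes, glow sphere and main explosion mesh.
# A system-level GlobalScale drives sizes, velocities, light radius and camera offsets.
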NiagaraSystem /BRCosmetics/Animation/Game/MainPlayer/Emotes/PowerFarmer/FX/NS_Emote_PowerFarmer_ExplosionSmall
dependencies {
 usedDataInterfaces [
  &NiagaraDataInterfaceCurve
  &NiagaraDataInterfaceVectorCurve
  &NiagaraDataInterfaceColorCurve
  &NiagaraDataInterfaceParticleRead
  &NiagaraDataInterfaceMeshRendererInfo
  &NiagaraDataInterfaceVector2DCurve
  &NiagaraDataInterfaceArrayFloat
  &NiagaraDataInterfaceCamera
 ]
 usedDynamicInputs []
 usedModules [
  &self:SystemSpawnScript.SystemScriptSource.SystemScriptGraph.NiagaraNodeAssignment_0.SetVariables_#1
  &SystemState
 ]
}
emitters {
 Blast_Stylized {
  bEnabled=false
  emitterUpdateScript [
   &EmitterState(Life Cycle Mode=::NewEnumerator1, Loop Behavior=::NewEnumerator1, Loop Delay=0.1383, Loop Duration=0.4, UseLoopDelay=true)
   &SpawnBurst_Instantaneous(Spawn Count=3, Use Spawn Probability=false)
  ]
  particleSpawnScript [
   &InitializeParticle(
    Color=<rgba 5 0.2879 0 1>
    Lifetime Max=0.3
    Lifetime Min=0.05
    Lifetime Mode=::NewEnumerator1
    Sprite Size=&Multiply_Vector2D_ByFloat(Float=$System.GlobalScale, Vector2D=&UniformRangedVector2D(Minimum=<35 35>, Maximum=<60 75>))
    Sprite Size Mode=::NewEnumerator4
    Sprite UV Mode=::NewEnumerator1
    UsePositionOffset=false
    Write Material Random=true
   )
   /Niagara/Modules/Spawn/Location/SphereLocation(
    MaskSpawnGroup=false
    Module.Override Randomness=false
    Sphere Radius=&Multiply_Float(A=4, B=$System.GlobalScale)
    Surface Only=true
   )
   &AddVelocityFromPoint(
    Module.Use Falloff Distance=false
    Velocity Strength=&Multiply_Float(A=&UniformRangedFloat(Minimum=200, Maximum=250), B=$System.GlobalScale)
   )
   &SubUVAnimation(End Frame=3, SubUV Animation Mode:ENiagara_SubUVLookupMode=::NewEnumerator2)
   &CameraOffset(Camera Offset Amount=&Multiply_Float(A=50, B=$System.GlobalScale))
  ]
  particleUpdateScript [
   &ParticleState()
   &Drag(
    Drag=&FloatFromCurve(FloatCurve=DataInterface() {propertyValues {Curve {keys [<at 0.2 1>,<at 0.3 0>]}}}, Scale Curve=15)
    Use Linear Drag=true
    Use Rotational Drag=false
   )
   &GravityForce(Gravity=&Multiply_VectorByFloat(
     Vector=<0 0 -2.2>
     Float=&FloatFromCurve(
      FloatCurve=DataInterface() {propertyValues {Curve {keys [{arriveTangent=3.1357, interpMode=RCIM_Cubic, leaveTangent=3.1357, tangentMode=RCTM_User, time=0.1, value=0}, <at 0.6 1>]}}}
      CurveIndex=$Emitter.Age
     )
    )) {enabled=false}
   &SolveForcesAndVelocity(Clamp Velocity=false, Limit Acceleration=false)
   &ScaleSpriteSize(Scale Factor=&Vector2DFromCurve(
     Vector2Curve=DataInterface() {propertyValues {
       XCurve {keys [
         <at 0 0.625 RCIM_Cubic>
         {arriveTangent=0.875, interpMode=RCIM_Cubic, leaveTangent=0.875, time=0.4, value=0.625}
         {arriveTangent=1.2531, interpMode=RCIM_Cubic, leaveTangent=1.2531, tangentMode=RCTM_User, time=1, value=1.3}
        ]}
       YCurve {keys [
         {arriveTangent=9.4573, interpMode=RCIM_Cubic, leaveTangent=9.4573, tangentMode=RCTM_User, time=0, value=0}
         <at 0.18 1.625>
         <at 1 2 RCIM_Cubic>
        ]}
      }}
     ScaleCurve=<1 1.4>
    ))
   &ScaleColor(
    Linear Color Curve=DataInterface() {propertyValues {
      AlphaCurve {keys [
        {arriveTangent=-0.3577, interpMode=RCIM_Cubic, leaveTangent=-0.3577, tangentMode=RCTM_User, time=0.08, value=1}
        {arriveTangent=-1.087, interpMode=RCIM_Cubic, leaveTangent=-1.0164, tangentMode=RCTM_Break, time=0.2616, value=1}
        <at 1 0>
       ]}
      BlueCurve {keys [<at 0.1976 0>]}
      GreenCurve {keys [<at 0.1976 0.1604>]}
      RedCurve {keys [<at 0.1976 4>]}
     }}
    Scale Mode=::NewEnumerator2
   )
   &DynamicMaterialParameters(
    Index 0 Param 1=&Multiply_Float(A=100, B=$System.GlobalScale)
    Index 0 Param 4=&FloatFromCurve(FloatCurve=DataInterface() {propertyValues {Curve {keys [{interpMode=RCIM_Cubic, tangentMode=RCTM_User, time=0, value=0}, {interpMode=RCIM_Cubic, tangentMode=RCTM_User, time=0.4, value=1}]}}})
    Index 1 Param 1=&FloatFromCurve(
     FloatCurve=DataInterface() {propertyValues {Curve {keys [
         {interpMode=RCIM_Cubic, tangentMode=RCTM_User, time=0.2, value=0}
         {arriveTangent=0.6295, interpMode=RCIM_Cubic, leaveTangent=0.6295, tangentMode=RCTM_User, time=1, value=1}
        ]}}}
     Scale Curve=0.03
    )
    Index 1 Param 2=&FloatFromCurve(
     FloatCurve=DataInterface() {propertyValues {Curve {keys [
         {arriveTangent=-9.7084, interpMode=RCIM_Cubic, leaveTangent=-9.7084, tangentMode=RCTM_User, time=0.1, value=50}
         {arriveTangent=-0.3869, interpMode=RCIM_Cubic, leaveTangent=-0.3869, tangentMode=RCTM_User, time=0.35, value=0}
        ]}}}
     CurveIndex=$Emitter.Age
     Scale Curve=&MaskFloatBySpawnGroup(Masked Float=0, Passthrough Float=0.9, Spawn Group Mask=1)
    )
    Index 1 Param 3=&FloatFromCurve(
     FloatCurve=DataInterface() {propertyValues {Curve {keys [{interpMode=RCIM_Cubic, tangentMode=RCTM_User, time=0.2, value=0.5}, {interpMode=RCIM_Cubic, tangentMode=RCTM_User, time=1, value=1}]}}}
     Scale Curve=2
    )
    Index 1 Param 4=&FloatFromCurve(FloatCurve=DataInterface() {propertyValues {Curve {keys [<at 0 1>,<at 1 0>]}}}, Scale Curve=5)
    Module.Param0WriteEnabled=true
    Module.Param0WriteEnabled1=true
    Module.Param1WriteEnabled=false
    Module.Param1WriteEnabled1=true
    Module.Param2WriteEnabled=false
    Module.Param2WriteEnabled1=true
    Module.Param3WriteEnabled=true
    Module.Param3WriteEnabled1=true
    Write Parameter Index 1=true
   )
  ]
  rendererClasses [&NiagaraSpriteRendererProperties]
  renderers [{rendererClass=&NiagaraSpriteRendererProperties}]
 }
 Dark_Chunks001 {
  simTarget=GPUComputeSim
  emitterUpdateScript [
   &EmitterState(Life Cycle Mode=::NewEnumerator1, Loop Behavior=::NewEnumerator1, Loop Delay=0.13, Loop Duration=0.1, UseLoopDelay=true)
   &SpawnBurst_Instantaneous(Spawn Count=18, Spawn Time=0.01, Use Spawn Probability=false, UseLoopCountLimit=false) {module=SpawnBurst_Instantaneous001}
   &SpawnBurst_Instantaneous(Spawn Count=15, Use Spawn Probability=false)
  ]
  particleSpawnScript [
   &InitializeParticle(
    Color=<rgba 0 0 0 5>
    Lifetime Max=0.15
    Lifetime Min=0.05
    Lifetime Mode=::NewEnumerator1
    Sprite Rotation Mode=::NewEnumerator2
    Sprite Size Mode=::NewEnumerator1
    Sprite UV Mode=::NewEnumerator3
    Uniform Sprite Size Max=25
    Uniform Sprite Size Min=20
    UsePositionOffset=false
    Write Material Random=true
   )
   &SphereLocation(
    MaskSpawnGroup=false
    Module.Override Randomness=false
    Sphere Radius=&Multiply_Float(A=25, B=$System.GlobalScale)
    Surface Only=false
   )
   &AddVelocity(
    Module.Use Falloff Distance=false
    Override Seed=false
    Velocity Mode=::NewEnumerator1
    Velocity Speed=&Multiply_Float(A=&RandomRangeFloat(Minimum=850, Maximum=1000), B=$System.GlobalScale)
   )
   &ApplyInitialForces(Apply Rotational Force to Rotational Velocity=false)
   /SetVariables(Particles.SpriteSize=&Multiply_Vector2D_ByFloat(Float=$System.GlobalScale, Vector2D=$StackContext.SpriteSize)) {isSetParameters=true, module=SetVariables_#2}
   &SubUVAnimation(SubUV Animation Mode:ENiagara_SubUVLookupMode=::NewEnumerator2)
   &CameraOffset(Camera Offset Amount=&Multiply_Float(A=10, B=$System.GlobalScale))
  ]
  particleUpdateScript [
   &ParticleState()
   &Drag(Drag=&RandomRangeFloat(Minimum=14, Maximum=15), Use Linear Drag=true, Use Rotational Drag=false)
   &CurlNoiseForce(Noise Frequency=7, Noise Quality / Cost=::NewEnumerator0, Noise Strength=2000, Pan Noise=false)
   &SolveForcesAndVelocity(Clamp Velocity=false, Limit Acceleration=false)
   &ScaleColor(
    Scale Alpha=&FloatFromCurve(FloatCurve=DataInterface() {propertyValues {Curve {keys [
         {arriveTangent=26.7273, interpMode=RCIM_Cubic, leaveTangent=26.7273, tangentMode=RCTM_User, time=-0.0092, value=0.0144}
         {interpMode=RCIM_Cubic, tangentMode=RCTM_User, time=0.1, value=1}
         {interpMode=RCIM_Cubic, tangentMode=RCTM_User, time=0.8, value=1}
         <at 1.0011 0 RCIM_Cubic>
        ]}}})
    ScaleA=true
    ScaleRGB=false
   )
   &ScaleSpriteSize(Scale Factor=&Vector2DFromFloat(Value=&FloatFromCurve(FloatCurve=DataInterface() {propertyValues {Curve {keys [<at -0.0046 0.3>,<at 0.1 1>,<at 0.66 1>,<at 1.0011 0.5>]}}})))
  ]
  rendererClasses [&NiagaraSpriteRendererProperties]
  renderers [{rendererClass=&NiagaraSpriteRendererProperties}]
 }
 Dark_Chunks_6 {
  simTarget=GPUComputeSim
  emitterUpdateScript [
   &EmitterState(Life Cycle Mode=::NewEnumerator1, Loop Behavior=::NewEnumerator1, Loop Delay=0.13, Loop Duration=0.1, UseLoopDelay=true)
   &SpawnBurst_Instantaneous(Spawn Count=18, Spawn Time=0.01, Use Spawn Probability=false, UseLoopCountLimit=false) {module=SpawnBurst_Instantaneous001}
   &SpawnBurst_Instantaneous(Spawn Count=15, Use Spawn Probability=false)
  ]
  particleSpawnScript [
   &InitializeParticle(
    Color=<rgba 0 0 0 5>
    Lifetime Max=0.15
    Lifetime Min=0.05
    Lifetime Mode=::NewEnumerator1
    Sprite Rotation Mode=::NewEnumerator2
    Sprite Size Mode=::NewEnumerator1
    Sprite UV Mode=::NewEnumerator3
    Uniform Sprite Size Max=25
    Uniform Sprite Size Min=20
    UsePositionOffset=false
    Write Material Random=true
   )
   &SphereLocation(
    MaskSpawnGroup=false
    Module.Override Randomness=false
    Sphere Radius=&Multiply_Float(A=25, B=$System.GlobalScale)
    Surface Only=false
   )
   &AddVelocity(
    Module.Use Falloff Distance=false
    Override Seed=false
    Velocity Mode=::NewEnumerator1
    Velocity Speed=&Multiply_Float(A=&RandomRangeFloat(Minimum=850, Maximum=1000), B=$System.GlobalScale)
   )
   &ApplyInitialForces(Apply Rotational Force to Rotational Velocity=false)
   /SetVariables(Particles.SpriteSize=&Multiply_Vector2D_ByFloat(Float=$System.GlobalScale, Vector2D=$StackContext.SpriteSize)) {isSetParameters=true, module=SetVariables_#2}
   &SubUVAnimation(SubUV Animation Mode:ENiagara_SubUVLookupMode=::NewEnumerator2)
   &CameraOffset(Camera Offset Amount=&Multiply_Float(A=10, B=$System.GlobalScale))
  ]
  particleUpdateScript [
   &ParticleState()
   &Drag(Drag=&RandomRangeFloat(Minimum=14, Maximum=15), Use Linear Drag=true, Use Rotational Drag=false)
   &CurlNoiseForce(Noise Frequency=7, Noise Quality / Cost=::NewEnumerator0, Noise Strength=2000, Pan Noise=false)
   &SolveForcesAndVelocity(Clamp Velocity=false, Limit Acceleration=false)
   &ScaleColor(
    Scale Alpha=&FloatFromCurve(FloatCurve=DataInterface() {propertyValues {Curve {keys [
         {arriveTangent=26.7273, interpMode=RCIM_Cubic, leaveTangent=26.7273, tangentMode=RCTM_User, time=-0.0092, value=0.0144}
         {interpMode=RCIM_Cubic, tangentMode=RCTM_User, time=0.1, value=1}
         {interpMode=RCIM_Cubic, tangentMode=RCTM_User, time=0.8, value=1}
         <at 1.0011 0 RCIM_Cubic>
        ]}}})
    ScaleA=true
    ScaleRGB=false
   )
   &ScaleSpriteSize(Scale Factor=&Vector2DFromFloat(Value=&FloatFromCurve(FloatCurve=DataInterface() {propertyValues {Curve {keys [<at -0.0046 0.3>,<at 0.1 1>,<at 0.66 1>,<at 1.0011 0.5>]}}})))
  ]
  rendererClasses [&NiagaraSpriteRendererProperties]
  renderers [{rendererClass=&NiagaraSpriteRendererProperties}]
 }
 Flash_wide {
  bEnabled=false
  emitterUpdateScript [
   &EmitterState(Life Cycle Mode=::NewEnumerator1, Loop Behavior=::NewEnumerator1, Loop Delay=0.0753, Loop Duration=0.1, UseLoopDelay=true)
   &SpawnBurst_Instantaneous(Spawn Count=1, Use Spawn Probability=false)
  ]
  particleSpawnScript [
   /Niagara/Modules/Spawn/Initialization/InitializeParticle(
    Color=<rgba 0 2.6633 8 0.8>
    Lifetime=0.14
    Mesh Scale=&RandomRangeVector(Minimum=<0.012 0.1 0.5>, Maximum=<0.012 2 0.1>)
    Module.Write Color=true
    Module.Write Lifetime=true
    Module.Write Mass=false
    Module.Write Material Random=true
    Module.Write Position=false
    Module.Write Scale=true
    Module.Write SpriteRotation=false
    Module.Write SpriteSize=false
    Position=&SimulationPosition()
    Sprite Size=<5 5>
   )
   &CameraOffset(Camera Offset Amount=&Multiply_Float(A=50, B=$System.GlobalScale))
   /SetVariables(Particles.Scale=&Multiply_VectorByFloat(Vector=$Particles.Scale, Float=&Multiply_Float(A=$System.GlobalScale, B=1))) {isSetParameters=true, module=SetVariables_#3}
  ]
  particleUpdateScript [
   &ParticleState()
   &SolveForcesAndVelocity(Clamp Velocity=false, Limit Acceleration=false)
   &ScaleMeshSize(Scale Factor=&VectorFromFloat(Value=&FloatFromCurve(FloatCurve=DataInterface() {propertyValues {Curve {keys [
          {arriveTangent=9.901, interpMode=RCIM_Cubic, leaveTangent=9.901, tangentMode=RCTM_User, time=0, value=0.4}
          {arriveTangent=12.2238, interpMode=RCIM_Cubic, leaveTangent=12.2238, tangentMode=RCTM_User, time=1, value=6}
         ]}}})))
   &ScaleColor(
    Scale Alpha=&FloatFromCurve(FloatCurve=DataInterface() {propertyValues {Curve {keys [<at 0.7 1 RCIM_Cubic>,<at 1 0>]}}})
    ScaleA=true
    ScaleRGB=false
   )
  ]
  rendererClasses [&NiagaraMeshRendererProperties]
  renderers [{rendererClass=&NiagaraMeshRendererProperties, rendererIndex=0}]
 }
 GlowSphereMesh {
  bEnabled=false
  emitterUpdateScript [
   &EmitterScalability(
    Module.bApplySpawnCountScale=false
    Module.bApplySpawnCountScaleByDistanceFraction=false
    Module.bUseMaxDistance=false
    Module.bUseMinDistance=false
   ) {enabled=false}
   &EmitterState(Life Cycle Mode=::NewEnumerator1, Loop Behavior=::NewEnumerator1, Loop Delay=0.1, Loop Duration=0.04, UseLoopDelay=true)
   &SpawnBurst_Instantaneous(Spawn Count=1, Use Spawn Probability=false)
  ]
  particleSpawnScript [
   &InitializeParticle(
    Color=&MultiplyLinearColor(Linear Color=$User.Global Color, Scale Factor=<3 3 3 1>)
    Lifetime=0.11
    Mesh Scale=<2 2 2>
    Module.Write Color=true
    Module.Write Lifetime=true
    Module.Write Mass=false
    Module.Write Material Random=true
    Module.Write Position=false
    Module.Write Scale=true
    Module.Write SpriteRotation=true
    Module.Write SpriteSize=true
    Position=&SimulationPosition()
    Sprite Rotation=&RandomRangeFloat(Minimum=0, Maximum=360)
    Sprite Size=&Multiply_Vector2D_ByFloat(Float=$System.GlobalScale, Vector2D=&Vector2DFromFloat(Value=250))
   )
   /SetVariables(Particles.Scale2:Vector3f=&Multiply_VectorByFloat(Vector=&Multiply_VectorByFloat(Vector=<10 10 10>, Float=$System.GlobalScale), Float=1)) {enabled=false, isSetParameters=true}
   &CameraOffset(Camera Offset Amount=&Multiply_Float(A=80, B=$System.GlobalScale))
   /SetVariables(Particles.Scale=&Multiply_VectorByFloat(Vector=$Particles.Scale, Float=$System.GlobalScale)) {enabled=false, isSetParameters=true}
  ]
  particleUpdateScript [
   &ParticleState()
   &DynamicMaterialParameters(
    Index 0 Param 1=&FloatFromCurve(FloatCurve=DataInterface() {propertyValues {Curve {keys [<at 0 0>,<at 0.22 0.5>,<at 1 0>]}}})
    Index 0 Param 2=&FloatFromCurve(FloatCurve=DataInterface() {propertyValues {Curve {keys [<at 0 15>,<at 1 0>]}}})
    Module.Param0WriteEnabled=true
    Module.Param1WriteEnabled=true
    Module.Param2WriteEnabled=true
    Module.Param3WriteEnabled=true
   )
   &SpriteRotationRate(
    Delta Time:NiagaraFloat=&FloatFromCurve(FloatCurve=DataInterface() {propertyValues {Curve {keys [
         {tangentMode=RCTM_User, time=0, value=1}
         {arriveTangent=-0.3174, interpMode=RCIM_Cubic, leaveTangent=-0.3174, tangentMode=RCTM_User, time=0.3, value=1}
         {arriveTangent=0.0123, interpMode=RCIM_Cubic, leaveTangent=0.0123, tangentMode=RCTM_User, time=1, value=0}
        ]}}})
    Rotation Rate=-5
   )
   &ScaleSpriteSize(
    Uniform Curve Scale=0.75
    Uniform Curve Sprite Scale=DataInterface() {propertyValues {Curve {keys [
        {arriveTangent=9.0725, interpMode=RCIM_Cubic, leaveTangent=9.0725, tangentMode=RCTM_User, time=0, value=0}
        {arriveTangent=-2.5249, interpMode=RCIM_Cubic, leaveTangent=-2.5249, tangentMode=RCTM_User, time=0.1408, value=0.9847}
       ]}}}
   )
   &ScaleMeshSize(Scale Factor=&VectorFromCurve(
     VectorCurve=DataInterface() {propertyValues {
       XCurve {keys [
         {arriveTangent=4.1785, interpMode=RCIM_Cubic, leaveTangent=4.1785, tangentMode=RCTM_User, time=0, value=0}
         {arriveTangent=0.9995, interpMode=RCIM_Cubic, leaveTangent=0.9995, tangentMode=RCTM_User, time=0.2223, value=0.6825}
         {arriveTangent=-0.0405, interpMode=RCIM_Cubic, leaveTangent=-0.0405, tangentMode=RCTM_User, time=1, value=1.575}
        ]}
       YCurve {keys [
         {arriveTangent=4.1785, interpMode=RCIM_Cubic, leaveTangent=4.1785, tangentMode=RCTM_User, time=0, value=0}
         {arriveTangent=0.9995, interpMode=RCIM_Cubic, leaveTangent=0.9995, tangentMode=RCTM_User, time=0.2223, value=0.6825}
         {arriveTangent=-0.0405, interpMode=RCIM_Cubic, leaveTangent=-0.0405, tangentMode=RCTM_User, time=1, value=1.575}
        ]}
       ZCurve {keys [
         {arriveTangent=4.1785, interpMode=RCIM_Cubic, leaveTangent=4.1785, tangentMode=RCTM_User, time=0, value=0}
         {arriveTangent=0.9995, interpMode=RCIM_Cubic, leaveTangent=0.9995, tangentMode=RCTM_User, time=0.2223, value=0.6825}
         {arriveTangent=-0.0405, interpMode=RCIM_Cubic, leaveTangent=-0.0405, tangentMode=RCTM_User, time=1, value=1.575}
        ]}
      }}
     Scale Curve=<0.6 0.6 0.5>
    ))
   &ScaleColor(
    Scale Alpha=&FloatFromCurve(FloatCurve=DataInterface() {propertyValues {Curve {keys [
         {interpMode=RCIM_Cubic, tangentMode=RCTM_User, time=-0.0046, value=0.1}
         {arriveTangent=-1.0036, interpMode=RCIM_Cubic, leaveTangent=-1.0036, tangentMode=RCTM_User, time=0.6, value=0.9}
         {arriveTangent=-2.4196, interpMode=RCIM_Cubic, leaveTangent=-2.4196, tangentMode=RCTM_User, time=1.0011, value=0.002}
        ]}}})
    ScaleA=true
    ScaleRGB=false
   )
   /SetVariables(Particles.Color2=&MakeLinearColorFromVectorAndFloat(
     "Vector (RGB)"=&Lerp_Vector(
      A=<25 5 0>
      B=<16 5 1>
      Alpha=&FloatFromCurve(FloatCurve=DataInterface() {propertyValues {Curve {keys [
           {arriveTangent=-1.9247, interpMode=RCIM_Cubic, time=0, value=1}
           {arriveTangent=-1.5152, interpMode=RCIM_Cubic, leaveTangent=-1.5152, time=0.33, value=0.25}
           {interpMode=RCIM_Cubic, tangentMode=RCTM_User, time=0.66, value=0}
          ]}}})
     )
     "Float (Alpha)"=&FloatFromCurve(
      FloatCurve=DataInterface() {propertyValues {Curve {keys [{arriveTangent=-5.1417, interpMode=RCIM_Cubic, time=0.66, value=1}, <at 1.0011 0 RCIM_Cubic>]}}}
      Scale Curve=0.02
     )
    )) {isSetParameters=true}
  ]
  rendererClasses [&NiagaraMeshRendererProperties, &NiagaraSpriteRendererProperties]
  renderers [
   {rendererClass=&NiagaraMeshRendererProperties, rendererIndex=0}
   {rendererClass=&NiagaraSpriteRendererProperties, rendererIndex=1}
  ]
 }
 Intro_Flash {
  bEnabled=false
  emitterUpdateScript [
   &EmitterState(Life Cycle Mode=::NewEnumerator1, Loop Behavior=::NewEnumerator1, Loop Delay=0.05, Loop Duration=0.1, UseLoopDelay=true)
   &SpawnBurst_Instantaneous(Spawn Count=1, Use Spawn Probability=false)
   &SpawnBurst_Instantaneous(Spawn Count=1, Spawn Group=1, Use Spawn Probability=false) {module=SpawnBurst_Instantaneous001}
  ]
  particleSpawnScript [
   &InitializeParticle(
    Color=&LerpLinearColors(
     StartColor=$User.Global Color
     EndColor=<rgba 0.9703 0.9339 1 1>
     LerpFactor=&MakeFloatFromInt(Int=$Emitter.SpawnGroup)
    )
    Lifetime=0.3
    Mesh Scale=&VectorFromFloat(Value=&Lerp_Float(A=0.35, B=0.25, Alpha=&MakeFloatFromInt(Int=&SpawnGroup())))
    Module.Write Color=true
    Module.Write Lifetime=true
    Module.Write Mass=false
    Module.Write Material Random=true
    Module.Write Position=false
    Module.Write Scale=true
    Module.Write SpriteRotation=false
    Module.Write SpriteSize=false
    Position=&SimulationPosition()
    Sprite Size=<5 5>
   )
   &InitialMeshOrientation(
    Mesh Orientation Mode=::NewEnumerator2
    Orientation Vector=&MaskVectorBySpawnGroup(
     Masked Vector=<0 0 1>
     Passthrough Vector=&RandomRangeVector(Minimum=<-1 -1 -1>, Maximum=<1 1 1>)
     Spawn Group Mask=1
    )
    Rotation=<0 0.25 0>
    Use Rotation Vector=true
   ) {module=InitialMeshOrientation001}
   &DynamicMaterialParameters(
    Index 0 Param 1=&UniformRangedFloat(Minimum=0, Maximum=20)
    Module.Param0WriteEnabled=true
    Module.Param1WriteEnabled=false
    Module.Param2WriteEnabled=false
    Module.Param3WriteEnabled=false
   )
   &CameraOffset(Camera Offset Amount=&Multiply_Float(A=15, B=$System.GlobalScale))
   /SetVariables(
    Particles.Scale=&Multiply_VectorByFloat(Vector=$Particles.Scale, Float=$System.GlobalScale)
    Particles.VisibilityTag=&MaskIntBySpawnGroup(Masked Int=1, Passthrough Int=0, Spawn Group Mask=1)
   ) {isSetParameters=true, module=SetVariables_#3}
  ]
  particleUpdateScript [
   &ParticleState()
   &ScaleMeshSize(Scale Factor=&VectorFromFloat(Value=&FloatFromCurve(
      FloatCurve=DataInterface() {propertyValues {Curve {keys [
          {arriveTangent=4.1785, interpMode=RCIM_Cubic, leaveTangent=4.1785, tangentMode=RCTM_User, time=0, value=0}
          {arriveTangent=0.9995, interpMode=RCIM_Cubic, leaveTangent=0.9995, tangentMode=RCTM_User, time=0.2, value=0.35}
          <at 0.4 0.95>
          {arriveTangent=-0.0405, interpMode=RCIM_Cubic, leaveTangent=-0.0405, tangentMode=RCTM_User, time=1, value=0.05}
         ]}}}
      Scale Curve=2.75
     )))
   &ScaleColor(
    Scale Alpha=&FloatFromCurve(FloatCurve=DataInterface() {propertyValues {Curve {keys [{arriveTangent=-3.5509, interpMode=RCIM_Cubic, leaveTangent=-3.5509, tangentMode=RCTM_User, time=0, value=1}, <at 1 0>]}}})
    Scale RGB=&MaskVectorBySpawnGroup(
     Masked Vector=<1 1 1>
     Passthrough Vector=&VectorFromCurve(
      VectorCurve=DataInterface() {propertyValues {XCurve {keys [<at 0 1>,<at 1 0.1>]}, YCurve {keys [<at 0 1>,<at 1 0.1>]}, ZCurve {keys [<at 0 1>,<at 1 0.1>]}}}
      Scale Curve=<50 5 5>
     )
    )
    ScaleA=true
    ScaleRGB=true
   )
   /SetVariables(
    Particles.PostControl:NiagaraFloat=&FloatFromCurve(FloatCurve=DataInterface() {propertyValues {Curve {keys [
         {arriveTangent=3.1121, interpMode=RCIM_Cubic, leaveTangent=3.1121, tangentMode=RCTM_User, time=0, value=0}
         {interpMode=RCIM_Cubic, tangentMode=RCTM_User, time=0.35, value=1}
         {tangentMode=RCTM_User, time=1, value=0}
        ]}}})
    Particles.PostON:NiagaraBool=&SetBoolByFloatComparison(
     A=1
     B=&FloatFromCurve(FloatCurve=DataInterface() {propertyValues {Curve {keys [<at 0 1.1>,<at 0.8 1.1>,<at 0.81 0>]}}})
     Comparison Type=::NewEnumerator6
    )
   ) {isSetParameters=true}
  ]
  rendererClasses [&NiagaraMeshRendererProperties, &NiagaraComponentRendererProperties]
  renderers [
   {rendererClass=&NiagaraMeshRendererProperties, rendererIndex=0}
   {rendererClass=&NiagaraComponentRendererProperties, rendererIndex=1}
  ]
 }
 Light {
  emitterUpdateScript [
   /SetVariables(Emitter.JustSphere=1) {isSetParameters=true}
   &EmitterState(Life Cycle Mode=::NewEnumerator1, Loop Behavior=::NewEnumerator1, Loop Delay=0.16, Loop Duration=1, UseLoopDelay=true)
   &SpawnBurst_Instantaneous(Spawn Count=1, Use Spawn Probability=false)
  ]
  particleSpawnScript [
   /Niagara/Modules/Spawn/Initialization/V2/InitializeParticle(Color Mode=::NewEnumerator0, Lifetime=0.12, Position Offset=<0 0 100>, UsePositionOffset=false, Write Material Random=true)
   &CameraOffset(Camera Offset Amount=&Multiply_Float(A=-50, B=$System.GlobalScale)) {enabled=false}
   /SetVariables(Particles.Scale=&Multiply_VectorByFloat(Vector=$Particles.Scale, Float=$System.GlobalScale)) {enabled=false, isSetParameters=true}
  ]
  particleUpdateScript [
   &ParticleState()
   &ScaleColor(
    Linear Color Curve=DataInterface() {propertyValues {
      AlphaCurve {keys [<at 0 3>,<at 0.1459 3>,<at 0.6684 1>,<at 1 0>]}
      BlueCurve {keys [<at 0.0129 1>]}
      GreenCurve {keys [<at 0.0129 1>]}
      RedCurve {keys [<at 0.0129 1>]}
     }}
    Scale Mode=::NewEnumerator2
   ) {enabled=false}
   &ScaleMeshSize(Scale Factor=&VectorFromFloat(Value=&FloatFromCurve(FloatCurve=DataInterface() {propertyValues {Curve {keys [
          {arriveTangent=10.3331, interpMode=RCIM_Cubic, leaveTangent=10.3331, tangentMode=RCTM_User, time=0, value=0.2}
          <at 0.2 0.95>
          {tangentMode=RCTM_User, time=1, value=1.2}
         ]}}}))) {enabled=false}
   &DynamicMaterialParameters(
    Index 0 Param 1=&FloatFromCurve(FloatCurve=DataInterface() {propertyValues {Curve {keys [<at 0 1>,<at 1 0.1>]}}}, Scale Curve=5)
    Index 0 Param 2=4
    Index 0 Param 3=&Multiply_Float(A=150, B=$System.GlobalScale)
    Index 0 Param 4=&FloatFromCurve(FloatCurve=DataInterface() {propertyValues {Curve {keys [
         {interpMode=RCIM_Cubic, tangentMode=RCTM_User, time=0, value=1}
         <at 0.1 0 RCIM_Cubic>
         {interpMode=RCIM_Cubic, tangentMode=RCTM_User, time=0.5, value=1}
        ]}}})
    Module.Param0WriteEnabled=true
    Module.Param1WriteEnabled=true
    Module.Param2WriteEnabled=true
    Module.Param3WriteEnabled=true
   ) {enabled=false}
   &FN_Light_v3(
    Color=&ColorFromCurve(ColorCurve=DataInterface() {propertyValues {
       AlphaCurve {keys [<at 0 1 RCIM_Cubic>,<at 1 0 RCIM_Cubic>]}
       BlueCurve {keys [
         <at 0.0317 1.3875>
         {arriveTangent=-1.7476, interpMode=RCIM_Cubic, leaveTangent=-1.7476, time=0.0984, value=0}
         <at 0.7899 0.0625>
         <at 1 0 RCIM_Cubic>
        ]}
       GreenCurve {keys [
         <at 0.0317 2.5184>
         {arriveTangent=-2.4393, interpMode=RCIM_Cubic, leaveTangent=-2.4393, time=0.0984, value=0.7736}
         <at 0.7899 0.669>
         <at 1 0 RCIM_Cubic>
        ]}
       RedCurve {keys [<at 0.0317 5>, {arriveTangent=-3.957, interpMode=RCIM_Cubic, leaveTangent=-3.957, time=0.0984, value=5}, <at 0.7899 2>, <at 1 0 RCIM_Cubic>]}
      }})
    Create Color Attribute=true
    Create Exponent Attribute=true
    Create Position Attribute=true
    Create Radius Attribute=true
    Enable Particle Light Enable=true
    Exponent=1.5
    Light Enabled=true
    Position:Vector3f=&Multiply_VectorByFloat(
     Vector=<0 0 25>
     Float=&Multiply_Float(
      A=$System.GlobalScale
      B=&FloatFromCurve(FloatCurve=DataInterface() {propertyValues {Curve {keys [{interpMode=RCIM_Cubic, tangentMode=RCTM_User, time=0, value=1.25}, <at 1 1.5>]}}})
     )
    )
    Radius:NiagaraFloat=&Multiply_Float(A=250, B=$System.GlobalScale)
   )
  ]
  rendererClasses [&NiagaraLightRendererProperties]
  renderers [{rendererClass=&NiagaraLightRendererProperties, rendererIndex=0}]
 }
 NE_Explosion_Main {
  bEnabled=false
  emitterUpdateScript [
   &EmitterState(Life Cycle Mode=::NewEnumerator1, Loop Behavior=::NewEnumerator1, Loop Delay=0.1383, Loop Duration=1.6229, UseLoopDelay=true)
   &SpawnBurst_Instantaneous(Spawn Group=1, Use Spawn Probability=false, UseLoopCountLimit=false) {module=SpawnBurst_Instantaneous002}
   &SpawnBurst_Instantaneous(Spawn Count=5, Use Spawn Probability=false, UseLoopCountLimit=false) {module=SpawnBurst_Instantaneous001}
  ]
  particleSpawnScript [
   &InitializeParticle(
    Color=&MaskLinearColorBySpawnGroup(Masked Linear Color=<rgba 0.2 0.2 0.2 1>, Passthrough Linear Color=<rgba 0.2 0.2 0.2 1>, Spawn Group Mask=1)
    Lifetime=&MaskFloatBySpawnGroup(
     Masked Float=2.6
     Passthrough Float=&SelectFloatFromArray(
      Array Sampling Mode=::NewEnumerator0
      Direct Array Index=?
      Float Selection Array=DataInterface() {propertyValues {FloatData [1,1]}}
      Random Seed=0
     )
     Spawn Group Mask=1
    )
    Mesh Scale=&Multiply_VectorByFloat(
     Vector=&MaskVectorBySpawnGroup(
      Masked Vector=<1.5 1.75 1.3>
      Passthrough Vector=&RandomRangeVector(Minimum=<1.5 1.4 1.4>, Maximum=<1.5 1.75 1.75>)
      Spawn Group Mask=1
     )
     Float=&MaskFloatBySpawnGroup(
      Masked Float=1.5
      Passthrough Float=&FloatFromCurve(
       FloatCurve=DataInterface() {propertyValues {Curve {keys [<at 0 0.9 RCIM_Constant>,<at 0.5 1 RCIM_Constant>,<at 1 1>]}}}
       CurveIndex=&RandomRangeFloat(Minimum=0, Maximum=1)
      )
      Spawn Group Mask=1
     )
    )
    Mesh Scale Mode=::NewEnumerator4
    UsePositionOffset=false
    Write Material Random=true
   )
   &ShapeLocation(
    MaskSpawnGroup=false
    Override Seed=false
    Sphere Radius=&Multiply_Float(A=100, B=$System.GlobalScale)
    UseOwnerScale=true
   )
   &AddVelocity(
    Module.Use Falloff Distance=false
    Origin Offset=&Multiply_VectorByFloat(Vector=<0 0 -25>, Float=$System.GlobalScale)
    Origin Offset Coordinate Space=::World
    Override Seed=false
    Velocity Mode=::NewEnumerator1
    Velocity Speed=&MaskFloatBySpawnGroup(
     Masked Float=0
     Passthrough Float=&Multiply_Float(A=&RandomRangeFloat(Minimum=1100, Maximum=1350), B=$System.GlobalScale)
     Spawn Group Mask=1
    )
   )
   &CameraOffset(Camera Offset Amount=&Multiply_Float(A=&MaskFloatBySpawnGroup(Masked Float=175, Passthrough Float=100, Spawn Group Mask=1), B=$System.GlobalScale))
   /SetVariables(Particles.Scale=&Multiply_VectorByFloat(Vector=$Particles.Scale, Float=&Multiply_Float(A=2.5, B=$System.GlobalScale))) {isSetParameters=true}
  ]
  particleUpdateScript [
   &ParticleState()
   &GravityForce(Gravity=&Multiply_VectorByFloat(
     Vector=&MaskVectorBySpawnGroup(
      Masked Vector=<0 0 0>
      Passthrough Vector=&RandomRangeVector(Minimum=<0 0 1250>, Maximum=<0 0 1500>)
      Spawn Group Mask=1
     )
     Float=$System.GlobalScale
    ))
   &Drag(Drag=&RandomRangeFloat(Minimum=9, Maximum=10), Use Linear Drag=true, Use Rotational Drag=false)
   &DynamicMaterialParameters(
    Index 0 Param 1=&FloatFromCurve(
     FloatCurve=DataInterface() {propertyValues {Curve {keys [{arriveTangent=-2.2221, interpMode=RCIM_Cubic, leaveTangent=-2.2221, tangentMode=RCTM_User, time=0.1, value=1}, <at 0.95 0>]}}}
     Scale Curve=&Multiply_Float(A=-200, B=$System.GlobalScale)
    )
    Index 0 Param 2=&FloatFromCurve(
     FloatCurve=DataInterface() {propertyValues {Curve {keys [
         {tangentMode=RCTM_User, time=0, value=1}
         {arriveTangent=-0.0405, interpMode=RCIM_Cubic, leaveTangent=-0.0405, tangentMode=RCTM_User, time=0.2, value=1}
         {arriveTangent=-0.0002, interpMode=RCIM_Cubic, leaveTangent=-0.0002, tangentMode=RCTM_User, time=0.9, value=0.6}
        ]}}}
     Scale Curve=0.7
    )
    Index 0 Param 3=&FloatFromCurve(
     FloatCurve=DataInterface() {propertyValues {Curve {keys [
         {arriveTangent=1.4725, interpMode=RCIM_Cubic, leaveTangent=1.4725, tangentMode=RCTM_User, time=0, value=0}
         {tangentMode=RCTM_User, time=1, value=1}
        ]}}}
     Scale Curve=&Multiply_Float(
      A=&MaskFloatBySpawnGroup(Masked Float=875, Passthrough Float=&RandomRangeFloat(Minimum=875, Maximum=1000), Spawn Group Mask=1)
      B=$System.GlobalScale
     )
    )
    Index 0 Param 4=&FloatFromCurve(
     FloatCurve=DataInterface() {propertyValues {Curve {
        preInfinityExtrap=RCCE_Linear
        keys [
         {arriveTangent=0.1175, interpMode=RCIM_Cubic, leaveTangent=16.4114, tangentMode=RCTM_Break, time=0.03, value=-3}
         {arriveTangent=2.6068, interpMode=RCIM_Cubic, leaveTangent=2.6068, tangentMode=RCTM_User, time=0.3881, value=-0.5}
         {arriveTangent=32.7533, interpMode=RCIM_Cubic, leaveTangent=14.4841, tangentMode=RCTM_Break, time=0.6465, value=2.75}
         {interpMode=RCIM_Cubic, tangentMode=RCTM_User, time=0.8319, value=11.25}
        ]
       }}}
     Scale Curve=0.2
    )
    Index 1 Param 1=&FloatFromCurve(
     FloatCurve=DataInterface() {propertyValues {Curve {keys [
         {arriveTangent=0.5348, interpMode=RCIM_Cubic, leaveTangent=0.5348, tangentMode=RCTM_User, time=0, value=0.01}
         {arriveTangent=1.457, interpMode=RCIM_Cubic, leaveTangent=1.457, tangentMode=RCTM_User, time=0.04, value=1}
         {arriveTangent=85.7775, interpMode=RCIM_Cubic, leaveTangent=85.7775, tangentMode=RCTM_User, time=0.08, value=16}
        ]}}}
     Scale Curve=&MaskFloatBySpawnGroup(Masked Float=5, Passthrough Float=&RandomRangeFloat(Minimum=5, Maximum=5), Spawn Group Mask=1)
    )
    Index 1 Param 2=&Divide_Float(
     A=&MaskFloatBySpawnGroup(
      Masked Float=0.0056
      Passthrough Float=&RandomRangeFloat(Minimum=0.0048, Maximum=0.0056)
      Spawn Group Mask=1
     )
     B=$System.GlobalScale
    )
    Index 1 Param 3=&FloatFromCurve(
     FloatCurve=DataInterface() {propertyValues {Curve {keys [{interpMode=RCIM_Cubic, tangentMode=RCTM_User, time=0.4, value=0.5}, <at 0.8 0.1>]}}}
     Scale Curve=5
    )
    Index 1 Param 4=&FloatFromCurve(FloatCurve=DataInterface() {propertyValues {Curve {keys [{interpMode=RCIM_Cubic, tangentMode=RCTM_User, time=0.05, value=1}, {interpMode=RCIM_Cubic, tangentMode=RCTM_User, time=1, value=6}]}}})
    Index 2 Param 1=&FloatFromCurve(
     FloatCurve=DataInterface() {propertyValues {Curve {keys [
         {arriveTangent=1.5203, interpMode=RCIM_Cubic, leaveTangent=1.5203, tangentMode=RCTM_User, time=0, value=0}
         {arriveTangent=0.2818, interpMode=RCIM_Cubic, leaveTangent=0.2818, tangentMode=RCTM_User, time=1, value=1.1}
        ]}}}
     Scale Curve=1.5
    )
    Index 2 Param 2=&FloatFromCurve(
     FloatCurve=DataInterface() {propertyValues {Curve {keys [
         {interpMode=RCIM_Cubic, tangentMode=RCTM_User, time=0.2, value=0}
         {arriveTangent=2.4333, interpMode=RCIM_Cubic, leaveTangent=2.4333, tangentMode=RCTM_User, time=0.9, value=0.8}
        ]}}}
     Scale Curve=&Multiply_Float(A=-0.325, B=$System.GlobalScale)
    )
    Index 2 Param 3=&FloatFromCurve(
     FloatCurve=DataInterface() {propertyValues {Curve {keys [
         {arriveTangent=-41.6084, interpMode=RCIM_Cubic, leaveTangent=-41.6084, tangentMode=RCTM_User, time=0.04, value=1}
         {arriveTangent=-0.3021, interpMode=RCIM_Cubic, leaveTangent=-0.3021, tangentMode=RCTM_User, time=0.06, value=0}
        ]}}}
     Scale Curve=100
    )
    Index 2 Param 4=&FloatFromCurve(FloatCurve=DataInterface() {propertyValues {Curve {keys [{interpMode=RCIM_Cubic, tangentMode=RCTM_User, time=0, value=1}, {interpMode=RCIM_Cubic, tangentMode=RCTM_User, time=0.5, value=1}]}}})
    Index 3 Param 1=&FloatFromCurve(
     FloatCurve=DataInterface() {propertyValues {Curve {keys [
         {arriveTangent=2.2353, interpMode=RCIM_Cubic, leaveTangent=2.2353, tangentMode=RCTM_User, time=0, value=0.2}
         {arriveTangent=1.4416, interpMode=RCIM_Cubic, leaveTangent=1.4416, tangentMode=RCTM_User, time=0.2, value=0.5}
         {interpMode=RCIM_Cubic, tangentMode=RCTM_User, time=0.8, value=1.25}
        ]}}}
     Scale Curve=-2.5
    )
    Index 3 Param 2=&FloatFromCurve(
     FloatCurve=DataInterface() {propertyValues {Curve {keys [
         {arriveTangent=5.1563, interpMode=RCIM_Cubic, leaveTangent=5.1563, tangentMode=RCTM_User, time=0, value=0}
         {arriveTangent=0.1309, interpMode=RCIM_Cubic, leaveTangent=0.1309, tangentMode=RCTM_User, time=0.1699, value=0.4}
         {arriveTangent=0.8545, interpMode=RCIM_Cubic, leaveTangent=0.8545, tangentMode=RCTM_User, time=1, value=1}
        ]}}}
     Scale Curve=-0.4
    )
    Index 3 Param 3=&FloatFromCurve(FloatCurve=DataInterface() {propertyValues {Curve {keys [
         {interpMode=RCIM_Cubic, tangentMode=RCTM_User, time=0.01, value=16}
         {arriveTangent=-5.9757, interpMode=RCIM_Cubic, leaveTangent=-5.9757, tangentMode=RCTM_User, time=0.1, value=2.5}
         {arriveTangent=-5.9757, interpMode=RCIM_Cubic, leaveTangent=-5.9757, tangentMode=RCTM_User, time=0.15, value=1.25}
        ]}}})
    Index 3 Param 4=&FloatFromCurve(
     FloatCurve=DataInterface() {propertyValues {Curve {keys [{tangentMode=RCTM_User, time=0, value=0}, {tangentMode=RCTM_User, time=1, value=1}]}}}
     Scale Curve=-100
    )
    Module.Param0WriteEnabled=true
    Module.Param0WriteEnabled1=true
    Module.Param0WriteEnabled2=true
    Module.Param0WriteEnabled3=true
    Module.Param1WriteEnabled=true
    Module.Param1WriteEnabled1=true
    Module.Param1WriteEnabled2=true
    Module.Param1WriteEnabled3=true
    Module.Param2WriteEnabled=true
    Module.Param2WriteEnabled1=true
    Module.Param2WriteEnabled2=true
    Module.Param2WriteEnabled3=true
    Module.Param3WriteEnabled=true
    Module.Param3WriteEnabled1=true
    Module.Param3WriteEnabled2=true
    Module.Param3WriteEnabled3=true
    Write Parameter Index 1=true
    Write Parameter Index 2=true
    Write Parameter Index 3=true
   )
   &ScaleColor(
    Curve Index=&Lerp_Float(A=$Particles.NormalizedAge, B=$Particles.NormalizedAge, Alpha=0.4)
    Linear Color Curve=DataInterface() {propertyValues {
      AlphaCurve {keys [<at 0 1.2>,<at 0.073 1.26>,<at 0.4927 1>,<at 0.7681 1.26>,<at 1 0>]}
      BlueCurve {keys [<at 0.0178 0>,<at 0.041 0.6042>,<at 0.4977 0.4497>]}
      GreenCurve {keys [<at 0.0178 1.751>,<at 0.041 0.6042>,<at 0.4977 0.4497>]}
      RedCurve {keys [<at 0.0178 40>,<at 0.041 0.6042>,<at 0.4977 0.4497>]}
     }}
    Scale Mode=::NewEnumerator2
   )
   &CurlNoiseForce(
    Noise Frequency=&Divide_Float(A=0.4, B=$System.GlobalScale)
    Noise Quality / Cost=::NewEnumerator0
    Noise Strength=&FloatFromCurve(
     FloatCurve=DataInterface() {propertyValues {Curve {keys [
         {arriveTangent=3.7261, interpMode=RCIM_Cubic, leaveTangent=3.7261, tangentMode=RCTM_User, time=0.05, value=0}
         {interpMode=RCIM_Cubic, tangentMode=RCTM_User, time=0.5, value=1}
         <at 0.8 0>
        ]}}}
     Scale Curve=&Multiply_Float(A=2500, B=$System.GlobalScale)
    )
    Pan Noise=true
    Pan Noise Field=&RandomRangeVector(Minimum=<0 0 0>, Maximum=<1 1 1>)
   )
   &SolveForcesAndVelocity(Clamp Velocity=false, Limit Acceleration=false)
   &ScaleMeshSize(Scale Factor=&VectorFromCurve(
     VectorCurve=DataInterface() {propertyValues {
       XCurve {keys [
         {arriveTangent=4.1785, interpMode=RCIM_Cubic, leaveTangent=4.1785, tangentMode=RCTM_User, time=0, value=0.5}
         {arriveTangent=0.581, interpMode=RCIM_Cubic, leaveTangent=0.581, tangentMode=RCTM_User, time=0.1, value=0.775}
         {arriveTangent=0.0837, interpMode=RCIM_Cubic, leaveTangent=0.0837, tangentMode=RCTM_User, time=1, value=1.1}
        ]}
       YCurve {keys [
         {arriveTangent=4.1785, interpMode=RCIM_Cubic, leaveTangent=4.1785, tangentMode=RCTM_User, time=0, value=0.5}
         {arriveTangent=0.581, interpMode=RCIM_Cubic, leaveTangent=0.581, tangentMode=RCTM_User, time=0.1, value=0.775}
         {arriveTangent=0.0837, interpMode=RCIM_Cubic, leaveTangent=0.0837, tangentMode=RCTM_User, time=1, value=1.1}
        ]}
       ZCurve {keys [
         {arriveTangent=4.1785, interpMode=RCIM_Cubic, leaveTangent=4.1785, tangentMode=RCTM_User, time=0, value=0.5}
         {arriveTangent=0.8594, interpMode=RCIM_Cubic, leaveTangent=0.8594, tangentMode=RCTM_User, time=0.0803, value=0.7813}
         {arriveTangent=0.0837, interpMode=RCIM_Cubic, leaveTangent=0.0837, tangentMode=RCTM_User, time=1, value=1.1}
        ]}
      }}
     Scale Curve=&Multiply_VectorByFloat(
      Vector=<0.6 0.6 0.6>
      Float=&FloatFromCurve(
       FloatCurve=DataInterface() {propertyValues {Curve {keys [<at 0.5 0.65 RCIM_Constant>,<at 0.7 0.7 RCIM_Constant>]}}}
       CurveIndex=&RandomRangeFloat(Minimum=0.4, Maximum=1)
       Scale Curve=&MaskFloatBySpawnGroup(
        Masked Float=&FloatFromCurve(
         FloatCurve=DataInterface() {propertyValues {Curve {keys [
             {arriveTangent=0.8977, interpMode=RCIM_Cubic, leaveTangent=0.8977, tangentMode=RCTM_User, time=0, value=0.6}
             {arriveTangent=0.4659, interpMode=RCIM_Cubic, leaveTangent=0.4659, tangentMode=RCTM_User, time=0.7, value=1}
            ]}}}
         Scale Curve=0.88
        )
        Passthrough Float=1
        Spawn Group Mask=1
       )
      )
     )
    ))
   &KillParticles(Kill Particles=&SetBoolByFloatComparison(
     A=1
     B=&FloatFromCurve(FloatCurve=DataInterface() {propertyValues {Curve {keys [<at 1.5 0>,<at 2.25 1.01>]}}}, CurveIndex=$Emitter.Age)
     Comparison Type=::NewEnumerator6
    ))
  ]
  rendererClasses [&NiagaraMeshRendererProperties]
  renderers [{rendererClass=&NiagaraMeshRendererProperties, rendererIndex=0}]
 }
 Ribbon001 {
  bEnabled=false
  emitterUpdateScript [&EmitterState()]
  particleSpawnScript [&InitializeRibbon(
    Color=<rgba 0.1441 0.131 0.1206 0.9>
    Lifetime=1.5
    Ribbon Width=&Multiply_Float(A=10, B=$System.GlobalScale)
    UsePositionOffset=false
    Write Material Random=true
   )]
  particleUpdateScript [
   &ParticleState()
   &CurlNoiseForce(
    Noise Frequency=&FloatFromCurve(FloatCurve=DataInterface() {propertyValues {Curve {keys [<at 0.5 1>,<at 1 0.3>]}}}, Scale Curve=20)
    Noise Quality / Cost=::NewEnumerator0
    Noise Strength=&FloatFromCurve(
     FloatCurve=DataInterface() {propertyValues {Curve {keys [{tangentMode=RCTM_User, time=0.5, value=1}, {tangentMode=RCTM_User, time=1, value=2}]}}}
     Scale Curve=100
    )
    Pan Noise=false
   )
   &SolveForcesAndVelocity(Clamp Velocity=false, Limit Acceleration=false)
   &ScaleColor(
    Linear Color Curve=DataInterface() {propertyValues {
      AlphaCurve {keys [<at 0.7526 1>,<at 1 0.1>]}
      BlueCurve {keys [<at 0.1 0.3177>,<at 1 0.0469>]}
      GreenCurve {keys [<at 0.1 0.3177>,<at 1 0.0469>]}
      RedCurve {keys [<at 0.1 0.3177>,<at 1 0.0469>]}
     }}
    Scale Mode=::NewEnumerator2
   )
   &ScaleColor(
    Scale Alpha=&FloatFromCurve(
     FloatCurve=DataInterface() {propertyValues {Curve {keys [{interpMode=RCIM_Cubic, tangentMode=RCTM_User, time=0.6, value=1}, <at 1 0>]}}}
     CurveIndex=$Emitter.Age
    )
    ScaleA=true
    ScaleRGB=false
   ) {module=ScaleColor001}
   &ScaleRibbonWidth(Ribbon Width Scale=&FloatFromCurve(FloatCurve=DataInterface() {propertyValues {Curve {keys [{interpMode=RCIM_Cubic, tangentMode=RCTM_User, time=0.3, value=1}, {interpMode=RCIM_Cubic, tangentMode=RCTM_User, time=1, value=0.5}]}}}))
   &DynamicMaterialParameters(
    Index 0 Param 1=1
    Index 0 Param 2=20
    Index 0 Param 3=1
    Index 0 Param 4=1
    Module.Param0WriteEnabled=true
    Module.Param1WriteEnabled=true
    Module.Param2WriteEnabled=true
    Module.Param3WriteEnabled=true
   )
  ]
  rendererClasses [&NiagaraRibbonRendererProperties]
  renderers [{rendererClass=&NiagaraRibbonRendererProperties}]
 }
 Smoke_Ground {
  emitterUpdateScript [
   &EmitterState(Life Cycle Mode=::NewEnumerator1, Loop Behavior=::NewEnumerator1, Loop Delay=0.12, Loop Duration=0.2, UseLoopDelay=true)
   &SpawnBurst_Instantaneous(Spawn Count=3, Use Spawn Probability=false)
  ]
  particleSpawnScript [
   &InitializeParticle(
    Color Maximum=<rgba 0.025 0.0277 0.0304 2>
    Color Minimum=<rgba 0.1074 0.1302 0.126 1>
    Color Mode=::NewEnumerator2
    Lifetime=&FloatFromCurve(
     FloatCurve=DataInterface() {propertyValues {Curve {keys [<at 0 0.4>,<at 0.6056 0.75>,<at 0.6256 2.5 RCIM_Constant>]}}}
     CurveIndex=&RandomRangeFloat(Minimum=0, Maximum=1)
     Scale Curve=0.5
    )
    Sprite Rotation Mode=::NewEnumerator2
    Sprite Size Mode=::NewEnumerator3
    Sprite UV Mode=::NewEnumerator3
    Uniform Sprite Size=&Multiply_Float(A=&RandomRangeFloat(Minimum=150, Maximum=200), B=$System.GlobalScale)
    UsePositionOffset=false
    Write Material Random=true
   )
   &ShapeLocation(
    Handle Radius:NiagaraFloat=&Multiply_Float(A=25, B=$System.GlobalScale)
    Large Radius=&Multiply_Float(A=35, B=$System.GlobalScale)
    MaskSpawnGroup=false
    Rotation Coordinate Space=::World
    Shape Primitive=::NewEnumerator3
    Torus Distribution Mode:ENiagaraTorusDistributionMode=::NewEnumerator1
    U Position=&ReturnNormalizedExecIndex(Normalized Index Includes One=false)
    UseOwnerScale=true
   )
   &AddVelocityFromPoint(
    Module.Use Falloff Distance=false
    Origin Offset Coordinate Space=::World
    Velocity Origin:Vector3f=&SimulationPosition()
    Velocity Strength=&Multiply_Float(A=&UniformRangedFloat(Minimum=500, Maximum=600), B=$System.GlobalScale)
   )
   &ApplyInitialForces()
   &CameraOffset(Camera Offset Amount=&Multiply_Float(A=100, B=$System.GlobalScale))
   /SetVariables(
    Particles.Stored Particle Size:Vector3f=$Particles.Scale
    Particles.Stored Sprite Size:Vector2f=&Multiply_Vector2D_ByFloat(Float=$System.GlobalScale, Vector2D=$Particles.SpriteSize)
   ) {enabled=false, isSetParameters=true, module=SetVariables_#4}
  ]
  particleUpdateScript [
   &ParticleState()
   /SetVariables(Particles.Scale=$StackContext.Stored Particle Size, Particles.SpriteSize=$StackContext.Stored Sprite Size) {enabled=false, isSetParameters=true, module=SetVariables_#5}
   &AccelerationForce(Acceleration=&MakeVector(
     X=0
     Y=0
     Z=&Multiply_Float(A=&UniformRangedFloat(Minimum=-250, Maximum=500), B=$System.GlobalScale)
    ))
   &Drag(Drag=&RandomRangeFloat(Minimum=8, Maximum=13), Use Linear Drag=true, Use Rotational Drag=false)
   &self:ViewpointBased_Sprite_Rotation_02(Rotation Rate=3) {module=ViewpointBased_Sprite_Rotation_02}
   &SolveForcesAndVelocity(Clamp Velocity=false, Limit Acceleration=false)
   &ScaleSpriteSize(Uniform Curve Sprite Scale=DataInterface() {propertyValues {Curve {keys [
        {arriveTangent=4.1286, interpMode=RCIM_Cubic, leaveTangent=4.1286, tangentMode=RCTM_User, time=0, value=0.2}
        {arriveTangent=1.2171, interpMode=RCIM_Cubic, leaveTangent=1.2171, tangentMode=RCTM_User, time=0.309, value=0.83}
        <at 1 1.4 RCIM_Cubic>
       ]}}})
   &DynamicMaterialParameters(
    Index 0 Param 1=&FloatFromCurve(FloatCurve=DataInterface() {propertyValues {Curve {keys [{interpMode=RCIM_Cubic, tangentMode=RCTM_User, time=0, value=1}, {interpMode=RCIM_Cubic, tangentMode=RCTM_User, time=1, value=0}]}}})
    Index 0 Param 4=&FloatFromCurve(FloatCurve=DataInterface() {propertyValues {Curve {keys [{interpMode=RCIM_Cubic, tangentMode=RCTM_User, time=0, value=0}, {interpMode=RCIM_Cubic, tangentMode=RCTM_User, time=1, value=1}]}}})
    Module.Param0WriteEnabled=true
    Module.Param1WriteEnabled=false
    Module.Param2WriteEnabled=false
    Module.Param3WriteEnabled=true
   ) {module=DynamicMatParams}
   &ScaleColor(
    Scale Alpha=&FloatFromCurve(
     FloatCurve=DataInterface() {propertyValues {Curve {keys [
         {arriveTangent=6.7279, interpMode=RCIM_Cubic, leaveTangent=6.7279, tangentMode=RCTM_User, time=0.0025, value=0.5379}
         {arriveTangent=-0.5367, interpMode=RCIM_Cubic, leaveTangent=-0.5367, time=0.1843, value=0.9779}
         {arriveTangent=-0.9746, interpMode=RCIM_Cubic, leaveTangent=-0.9746, tangentMode=RCTM_User, time=1.0011, value=0.002}
        ]}}}
     Scale Curve=0.6
    )
    ScaleA=true
    ScaleRGB=true
   )
   &SubUVAnimation(SubUV Loop Count=0.5)
  ]
  rendererClasses [&NiagaraSpriteRendererProperties]
  renderers [{rendererClass=&NiagaraSpriteRendererProperties}]
 }
 Smoke_Ground001 {
  emitterUpdateScript [
   &EmitterState(Life Cycle Mode=::NewEnumerator1, Loop Behavior=::NewEnumerator1, Loop Delay=0.12, Loop Duration=0.2, UseLoopDelay=true)
   &SpawnBurst_Instantaneous(Spawn Count=7, Use Spawn Probability=false)
  ]
  particleSpawnScript [
   &InitializeParticle(
    Color Maximum=<rgba 0.025 0.0277 0.0304 2>
    Color Minimum=<rgba 0.1074 0.1302 0.126 1>
    Color Mode=::NewEnumerator2
    Lifetime=&FloatFromCurve(
     FloatCurve=DataInterface() {propertyValues {Curve {keys [<at 0 0.4>,<at 0.6056 0.75>,<at 0.6256 2.5 RCIM_Constant>]}}}
     CurveIndex=&RandomRangeFloat(Minimum=0, Maximum=1)
     Scale Curve=0.5
    )
    Sprite Rotation Mode=::NewEnumerator2
    Sprite Size Mode=::NewEnumerator3
    Sprite UV Mode=::NewEnumerator3
    Uniform Sprite Size=&Multiply_Float(A=&RandomRangeFloat(Minimum=100, Maximum=150), B=$System.GlobalScale)
    UsePositionOffset=false
    Write Material Random=true
   )
   &ShapeLocation(
    Handle Radius:NiagaraFloat=&Multiply_Float(A=25, B=$System.GlobalScale)
    Large Radius=&Multiply_Float(A=35, B=$System.GlobalScale)
    MaskSpawnGroup=false
    Rotation Coordinate Space=::World
    Shape Primitive=::NewEnumerator3
    Torus Distribution Mode:ENiagaraTorusDistributionMode=::NewEnumerator1
    U Position=&ReturnNormalizedExecIndex(Normalized Index Includes One=false)
    UseOwnerScale=true
   )
   &AddVelocityFromPoint(
    Module.Use Falloff Distance=false
    Origin Offset Coordinate Space=::World
    Velocity Origin:Vector3f=&SimulationPosition()
    Velocity Strength=&Multiply_Float(A=&UniformRangedFloat(Minimum=350, Maximum=400), B=$System.GlobalScale)
   )
   &ApplyInitialForces()
   &CameraOffset(Camera Offset Amount=&Multiply_Float(A=100, B=$System.GlobalScale))
   /SetVariables(
    Particles.Stored Particle Size:Vector3f=$Particles.Scale
    Particles.Stored Sprite Size:Vector2f=&Multiply_Vector2D_ByFloat(Float=$System.GlobalScale, Vector2D=$Particles.SpriteSize)
   ) {enabled=false, isSetParameters=true, module=SetVariables_#4}
  ]
  particleUpdateScript [
   &ParticleState()
   /SetVariables(Particles.Scale=$StackContext.Stored Particle Size, Particles.SpriteSize=$StackContext.Stored Sprite Size) {enabled=false, isSetParameters=true, module=SetVariables_#5}
   &AccelerationForce(Acceleration=&MakeVector(
     X=0
     Y=0
     Z=&Multiply_Float(A=&UniformRangedFloat(Minimum=-250, Maximum=500), B=$System.GlobalScale)
    ))
   &Drag(Drag=&RandomRangeFloat(Minimum=8, Maximum=13), Use Linear Drag=true, Use Rotational Drag=false)
   &self:ViewpointBased_Sprite_Rotation_01(Rotation Rate=3) {module=ViewpointBased_Sprite_Rotation_01}
   &SolveForcesAndVelocity(Clamp Velocity=false, Limit Acceleration=false)
   &ScaleSpriteSize(Uniform Curve Sprite Scale=DataInterface() {propertyValues {Curve {keys [
        {arriveTangent=4.1286, interpMode=RCIM_Cubic, leaveTangent=4.1286, tangentMode=RCTM_User, time=0, value=0.2}
        {arriveTangent=1.2171, interpMode=RCIM_Cubic, leaveTangent=1.2171, tangentMode=RCTM_User, time=0.309, value=0.83}
        <at 1 1.4 RCIM_Cubic>
       ]}}})
   &DynamicMaterialParameters(
    Index 0 Param 1=&FloatFromCurve(FloatCurve=DataInterface() {propertyValues {Curve {keys [<at -0.0046 1.0113>,<at 0.2 1>,<at 1.0011 -1>]}}})
    Index 0 Param 4=0.2
    Module.Param0WriteEnabled=true
    Module.Param1WriteEnabled=false
    Module.Param2WriteEnabled=false
    Module.Param3WriteEnabled=true
   ) {module=DynamicMatParams}
   &ScaleColor(
    Scale Alpha=&FloatFromCurve(
     FloatCurve=DataInterface() {propertyValues {Curve {keys [
         {arriveTangent=6.7279, interpMode=RCIM_Cubic, leaveTangent=6.7279, tangentMode=RCTM_User, time=0.0025, value=0.5379}
         {arriveTangent=-0.5367, interpMode=RCIM_Cubic, leaveTangent=-0.5367, time=0.1843, value=0.9779}
         {arriveTangent=-0.9746, interpMode=RCIM_Cubic, leaveTangent=-0.9746, tangentMode=RCTM_User, time=1.0011, value=0.002}
        ]}}}
     Scale Curve=2
    )
    ScaleA=true
    ScaleRGB=true
   )
   &SubUVAnimation(SubUV Loop Count=0.5)
  ]
  rendererClasses [&NiagaraSpriteRendererProperties]
  renderers [{rendererClass=&NiagaraSpriteRendererProperties}]
 }
 Smoke_Trail_Leader {
  simTarget=GPUComputeSim
  emitterUpdateScript [
   &EmitterState(
    Enable Distance Culling=true
    Life Cycle Mode=::NewEnumerator1
    Loop Behavior=::NewEnumerator1
    Loop Delay=0.0038
    Loop Duration=0.1
    Module.bUseMaxDistance=true
    Module.bUseMinDistance=false
    Scalability Mode=::NewEnumerator1
    UseLoopDelay=true
   )
   &SpawnBurst_Instantaneous(Use Spawn Probability=false, UseLoopCountLimit=false)
  ]
  particleSpawnScript [
   &InitializeParticle(
    Lifetime Max=1.75
    Lifetime Min=1.4
    Lifetime Mode=::NewEnumerator1
    Mass Max=1.4
    Mass Min=1
    Mass Mode=::NewEnumerator3
    Sprite Rotation Mode=::NewEnumerator2
    Sprite Size Mode=::NewEnumerator1
    Uniform Sprite Size Max=6
    Uniform Sprite Size Min=3
    UsePositionOffset=true
    Write Material Random=true
   )
   /Niagara/Modules/Spawn/Location/V2/SphereLocation(
    MaskSpawnGroup=false
    Non Uniform Scale=<1 1 0.5>
    Override Seed=false
    Sphere Radius=&Multiply_Float(A=1, B=$System.GlobalScale)
    Surface Only=false
    UseNonUniformScale=true
    UseOffset=false
    UseSphereOrientation=false
   )
   &AddVelocity(
    Module.Use Falloff Distance=false
    Origin Offset=&Multiply_VectorByFloat(Vector=<0 0 -1>, Float=$System.GlobalScale)
    Override Seed=false
    Velocity Mode=::NewEnumerator1
    Velocity Speed=&Multiply_Float(A=600, B=$System.GlobalScale)
   )
  ]
  particleUpdateScript [
   &ParticleState()
   &GravityForce(Gravity=&Multiply_VectorByFloat(Vector=<0 0 -2000>, Float=$System.GlobalScale))
   &Drag(Drag=3, Use Linear Drag=true, Use Rotational Drag=false)
   &ScaleColor(
    Scale Alpha=&FloatFromCurve(FloatCurve=DataInterface() {propertyValues {Curve {keys [<at -0.0046 1.0113 RCIM_Cubic>,<at 1.0011 0.002 RCIM_Cubic>]}}})
    ScaleA=true
    ScaleRGB=true
   )
   &SolveForcesAndVelocity(Clamp Velocity=false, Limit Acceleration=false)
  ]
  rendererClasses [&NiagaraSpriteRendererProperties, &NiagaraMeshRendererProperties]
  renderers [
   {rendererClass=&NiagaraSpriteRendererProperties}
   {rendererClass=&NiagaraMeshRendererProperties, rendererIndex=1}
  ]
 }
 Smoke_Trail_Leader001 {
  bEnabled=false
  emitterUpdateScript [
   &EmitterState(Life Cycle Mode=::NewEnumerator1, Loop Behavior=::NewEnumerator1, Loop Duration=2, UseLoopDelay=false)
   &SpawnBurst_Instantaneous(Spawn Count=2, Spawn Time=0.04, Use Spawn Probability=false, UseLoopCountLimit=false)
  ]
  particleSpawnScript [
   &InitializeParticle(
    Color=<rgba 0.5677 0.4858 0.4572 1>
    Lifetime Max=1.25
    Lifetime Min=0.625
    Lifetime Mode=::NewEnumerator1
    Mass Mode=::NewEnumerator3
    Mesh Scale Max=<0.25 0.25 0.25>
    Mesh Scale Min=<0.1 0.1 0.1>
    Mesh Scale Mode=::NewEnumerator2
    Position Offset=<0 0 15>
    Sprite Rotation Mode=::NewEnumerator2
    UsePositionOffset=true
    Write Material Random=true
   )
   &ShapeLocation(
    MaskSpawnGroup=false
    Override Seed=false
    Sphere Radius=&Multiply_Float(A=25, B=$System.GlobalScale)
    UseOwnerScale=true
   )
   &AddVelocity(
    Module.Use Falloff Distance=false
    Origin Offset=<0 0 -5>
    Override Seed=false
    Velocity Mode=::NewEnumerator1
    Velocity Speed=&Multiply_Float(A=&RandomRangeFloat(Minimum=650, Maximum=750), B=$System.GlobalScale)
   )
   &MeshRotationForce(
    Coordinate Space:ENiagaraMeshTransforms=::NewEnumerator2
    Rotation=&RandomRangeVector(Minimum=<-2 -2 -2>, Maximum=<3 3 3>)
   )
   /SetVariables() {enabled=false, isSetParameters=true}
   &ApplyInitialForces()
   /SetVariables(Particles.Scale=&Multiply_VectorByFloat(Vector=$Particles.Scale, Float=&Multiply_Float(A=1.25, B=$System.GlobalScale))) {isSetParameters=true}
  ]
  particleUpdateScript [
   &ParticleState()
   &UpdateMeshOrientation(Rotation Vector=<1 0 1>) {enabled=false}
   &GravityForce(Gravity=&Multiply_VectorByFloat(Vector=&RandomRangeVector(Minimum=<0 0 -2400>, Maximum=<0 0 -3000>), Float=$System.GlobalScale))
   &Drag(Drag=0.2, Use Linear Drag=true, Use Rotational Drag=false) {enabled=false}
   &ScaleColor(
    Scale Alpha=&FloatFromCurve(FloatCurve=DataInterface() {propertyValues {Curve {keys [<at -0.0046 1.0113 RCIM_Cubic>,<at 1.0011 0.002 RCIM_Cubic>]}}})
    ScaleA=true
    ScaleRGB=false
   )
   &ScaleMeshSize(Scale Factor=&VectorFromFloat(Value=&FloatFromCurve(FloatCurve=DataInterface() {propertyValues {Curve {keys [
          {tangentMode=RCTM_User, time=0, value=1}
          {arriveTangent=-0.0405, interpMode=RCIM_Cubic, leaveTangent=-0.0405, tangentMode=RCTM_User, time=0.7009, value=1}
          {arriveTangent=-8.0089, interpMode=RCIM_Cubic, leaveTangent=-8.0089, tangentMode=RCTM_User, time=1, value=0}
         ]}}})))
   &Collision(
    Age Colliding Particles=false
    Control Roll On Collision=true
    Friction=1
    Module.Control Rest State Via Surface Normal=false
    Module.EnableMaxCollisionCount=false
    Module.Randomize Collision Normal Vector=false
    Restitution=0.5
    ScaleRateOfRotation=false
   )
   &SolveForcesAndVelocity(Clamp Velocity=false, Limit Acceleration=false)
   &GenerateLocationEvent(Event Probability=1, Event Send Rate=24, Use Event Probability=false, UseEventDelay=false)
   /SetVariables(Particles.CollCount:NiagaraInt32=&Add_Integer(A=$Particles.CollCount, B=&MakeCustomIntFromBool(Bool=$Particles.HasCollided, True INT=1, False INT=0))) {enabled=false, isSetParameters=true}
   &KillParticles() {enabled=false}
  ]
  rendererClasses [&NiagaraSpriteRendererProperties, &NiagaraMeshRendererProperties]
  renderers [
   {rendererClass=&NiagaraSpriteRendererProperties}
   {rendererClass=&NiagaraMeshRendererProperties, rendererIndex=1}
  ]
 }
 Sparkles {
  emitterUpdateScript [
   &EmitterState(Life Cycle Mode=::NewEnumerator1, Loop Behavior=::NewEnumerator1, Loop Delay=0.0829, Loop Duration=1.5729, UseLoopDelay=true)
   &SpawnBurst_Instantaneous(Spawn Count=10, Use Spawn Probability=false, UseLoopCountLimit=false)
  ]
  particleSpawnScript [
   &InitializeParticle(
    Color=$User.Global Color
    Lifetime=&FloatFromCurve(
     FloatCurve=DataInterface() {propertyValues {Curve {keys [<at 0 0.25 RCIM_Constant>,<at 0.3 0.45>,<at 0.7 1 RCIM_Constant>,<at 0.85 1.5>,<at 1 2>]}}}
     CurveIndex=&RandomRangeFloat(Minimum=0, Maximum=1)
     Scale Curve=0.5
    )
    Sprite Size Max=<2 3>
    Sprite Size Min=<1 2>
    Sprite Size Mode=::NewEnumerator2
    UsePositionOffset=false
    Write Material Random=true
   )
   &ShapeLocation(Handle Radius=20, Large Radius=10, MaskSpawnGroup=false, Override Seed=false, Shape Primitive=::NewEnumerator3, UseOwnerScale=true)
   &AddVelocity(
    Module.Use Falloff Distance=false
    Origin Offset=<0 0 -100>
    Override Seed=false
    Velocity Mode=::NewEnumerator1
    Velocity Speed=&RandomRangeFloat(Minimum=700, Maximum=800)
   )
   &ApplyInitialForces(Apply Rotational Force to Rotational Velocity=false)
   /SetVariables(Particles.SpriteSize=&Multiply_Vector2D_ByFloat(Float=$System.GlobalScale, Vector2D=$StackContext.SpriteSize)) {isSetParameters=true, module=SetVariables_#2}
  ]
  particleUpdateScript [
   &ParticleState()
   &GravityForce(Gravity=&RandomRangeVector(Minimum=<0 0 -900>, Maximum=<0 0 -500>))
   &Drag(Drag=&RandomRangeFloat(Minimum=6, Maximum=8), Use Linear Drag=true, Use Rotational Drag=false)
   &Collision(
    Age Colliding Particles=false
    Module.Control Rest State Via Surface Normal=false
    Module.EnableMaxCollisionCount=false
    Module.Randomize Collision Normal Vector=false
   )
   &CurlNoiseForce(Noise Frequency=15, Noise Quality / Cost=::NewEnumerator0, Noise Strength=500, Pan Noise=false)
   &SolveForcesAndVelocity(Clamp Velocity=false, Limit Acceleration=false)
   &ScaleColor(
    Scale Alpha=&FloatFromCurve(FloatCurve=DataInterface() {propertyValues {Curve {keys [
         {arriveTangent=26.7273, interpMode=RCIM_Cubic, leaveTangent=26.7273, tangentMode=RCTM_User, time=0.1, value=0.0144}
         {interpMode=RCIM_Cubic, tangentMode=RCTM_User, time=0.15, value=1}
         {interpMode=RCIM_Cubic, tangentMode=RCTM_User, time=0.85, value=1}
         {arriveTangent=-1.3701, interpMode=RCIM_Cubic, leaveTangent=-1.3701, tangentMode=RCTM_User, time=1.0011, value=0}
        ]}}})
    Scale RGB=&Lerp_Vector(
     A=<0 0 0>
     B=<12.5 4 0.8>
     Alpha=&FloatFromCurve(FloatCurve=DataInterface() {propertyValues {Curve {keys [
          {arriveTangent=15.3906, interpMode=RCIM_Cubic, leaveTangent=15.3906, tangentMode=RCTM_User, time=-0.0092, value=1}
          {interpMode=RCIM_Cubic, tangentMode=RCTM_User, time=0.0967, value=0.8}
          {interpMode=RCIM_Cubic, tangentMode=RCTM_User, time=0.45, value=0.8}
          <at 0.7 0 RCIM_Cubic>
         ]}}})
    )
    ScaleA=true
    ScaleRGB=true
   )
   &ScaleSpriteSize(Scale Factor=&Vector2DFromFloat(Value=&FloatFromCurve(FloatCurve=DataInterface() {propertyValues {Curve {keys [<at -0.0046 0>,<at 0.025 1>,<at 0.66 1>,<at 1.0011 0.5>]}}}, Scale Curve=3.5)))
   &ScaleSpriteSizeBySpeed(Max Scale Factor=<0.5 2.2>, Min Scale Factor=<1 0.6>, Velocity Threshold=1500)
  ]
  rendererClasses [&NiagaraSpriteRendererProperties]
  renderers [{rendererClass=&NiagaraSpriteRendererProperties}]
 }
 Sparkles001 {
  bEnabled=false
  simTarget=GPUComputeSim
  emitterUpdateScript [
   &EmitterState(Life Cycle Mode=::NewEnumerator1, Loop Behavior=::NewEnumerator1, Loop Delay=0.0829, Loop Duration=0.1, UseLoopDelay=true)
   &SpawnParticlesFromOtherEmitter(
    Attribute Reader=DataInterface() {propertyValues {EmitterBinding {bindingMode=Other, emitterName=Smoke_Trail_Leader}}}
    Minimum Age Spawn Threshold=0
    Spawn Rate=500
    Spawn Rate Per Particle Cap=500
    Use Maximum Age=true
    Use Minimum Age=true
   )
  ]
  particleSpawnScript [
   &InitializeParticle(
    Color=$User.Global Color
    Lifetime=&FloatFromCurve(
     FloatCurve=DataInterface() {propertyValues {Curve {keys [<at 0 0.25 RCIM_Constant>,<at 0.3 0.45>,<at 0.7 1 RCIM_Constant>,<at 0.85 1.5>,<at 1 2>]}}}
     CurveIndex=&RandomRangeFloat(Minimum=0, Maximum=1)
     Scale Curve=0.5
    )
    Sprite Size Max=<4 5>
    Sprite Size Min=<3 3>
    Sprite Size Mode=::NewEnumerator2
    UsePositionOffset=false
    Write Material Random=true
   )
   /Niagara/Modules/Spawn/Location/SphereLocation(
    MaskSpawnGroup=false
    Module.Override Randomness=false
    Sphere Radius=&Multiply_Float(A=0, B=$System.GlobalScale)
    Surface Only=false
   ) {enabled=false}
   &AddVelocity(
    Module.Use Falloff Distance=false
    Override Seed=false
    Velocity Mode=::NewEnumerator1
    Velocity Speed=&Multiply_Float(A=&RandomRangeFloat(Minimum=1, Maximum=50), B=$System.GlobalScale)
   )
   &ApplyInitialForces(Apply Rotational Force to Rotational Velocity=false) {enabled=false}
   /SetVariables(Particles.SpriteSize=&Multiply_Vector2D_ByFloat(Float=$System.GlobalScale, Vector2D=$StackContext.SpriteSize)) {isSetParameters=true, module=SetVariables_#2}
   &SampleParticlesFromOtherEmitter(Applied Velocity Scale=1.4, Velocity Sampling=::NewEnumerator0)
  ]
  particleUpdateScript [
   &ParticleState()
   &GravityForce(Gravity=&RandomRangeVector(Minimum=<0 0 -900>, Maximum=<0 0 -500>))
   &Drag(Drag=&RandomRangeFloat(Minimum=2, Maximum=15), Use Linear Drag=true, Use Rotational Drag=false)
   &CurlNoiseForce(Noise Frequency=15, Noise Quality / Cost=::NewEnumerator0, Noise Strength=500, Pan Noise=false)
   &SolveForcesAndVelocity(Clamp Velocity=false, Limit Acceleration=false)
   &ScaleColor(
    Scale Alpha=&FloatFromCurve(FloatCurve=DataInterface() {propertyValues {Curve {keys [
         {arriveTangent=26.7273, interpMode=RCIM_Cubic, leaveTangent=26.7273, tangentMode=RCTM_User, time=-0.0092, value=0.0144}
         {interpMode=RCIM_Cubic, tangentMode=RCTM_User, time=0.1, value=1}
         {interpMode=RCIM_Cubic, tangentMode=RCTM_User, time=0.85, value=1}
         {arriveTangent=-1.3701, interpMode=RCIM_Cubic, leaveTangent=-1.3701, tangentMode=RCTM_User, time=1.0011, value=0}
        ]}}})
    Scale RGB=&Lerp_Vector(
     A=<0 0 0>
     B=<12.5 4 0.8>
     Alpha=&FloatFromCurve(FloatCurve=DataInterface() {propertyValues {Curve {keys [
          {arriveTangent=15.3906, interpMode=RCIM_Cubic, leaveTangent=15.3906, tangentMode=RCTM_User, time=-0.0092, value=1}
          {interpMode=RCIM_Cubic, tangentMode=RCTM_User, time=0.0967, value=0.8}
          {interpMode=RCIM_Cubic, tangentMode=RCTM_User, time=0.45, value=0.8}
          <at 0.7 0 RCIM_Cubic>
         ]}}})
    )
    ScaleA=true
    ScaleRGB=true
   )
   &ScaleSpriteSize(Scale Factor=&Vector2DFromFloat(Value=&FloatFromCurve(FloatCurve=DataInterface() {propertyValues {Curve {keys [<at -0.0046 0>,<at 0.025 1>,<at 0.66 1>,<at 1.0011 0.5>]}}}, Scale Curve=1.5)))
   &ScaleSpriteSizeBySpeed(Max Scale Factor=<0.5 5>, Min Scale Factor=<1 0.6>, Velocity Threshold=1250)
  ]
  rendererClasses [&NiagaraSpriteRendererProperties]
  renderers [{rendererClass=&NiagaraSpriteRendererProperties}]
 }
 Sparkles002 {
  emitterUpdateScript [
   &EmitterState(Life Cycle Mode=::NewEnumerator1, Loop Behavior=::NewEnumerator1, Loop Delay=0.0829, Loop Duration=1.5729, UseLoopDelay=true)
   &SpawnBurst_Instantaneous(Spawn Count=100, Use Spawn Probability=false, UseLoopCountLimit=false)
  ]
  particleSpawnScript [
   &InitializeParticle(
    Color=$User.Global Color
    Lifetime=&FloatFromCurve(
     FloatCurve=DataInterface() {propertyValues {Curve {keys [<at 0 0.25 RCIM_Constant>,<at 0.3 0.45>,<at 0.7 1 RCIM_Constant>,<at 0.85 1.5>,<at 1 2>]}}}
     CurveIndex=&RandomRangeFloat(Minimum=0, Maximum=1)
     Scale Curve=0.5
    )
    Sprite Size Max=<2 3>
    Sprite Size Min=<1 2>
    Sprite Size Mode=::NewEnumerator2
    UsePositionOffset=false
    Write Material Random=true
   )
   &ShapeLocation(Handle Radius=10, Large Radius=5, MaskSpawnGroup=false, Override Seed=false, Shape Primitive=::NewEnumerator3, UseOwnerScale=true)
   &AddVelocity(
    Module.Use Falloff Distance=false
    Origin Offset=<0 0 -50>
    Override Seed=false
    Velocity Mode=::NewEnumerator1
    Velocity Speed=&RandomRangeFloat(Minimum=500, Maximum=600)
   )
   &ApplyInitialForces(Apply Rotational Force to Rotational Velocity=false)
   /SetVariables(Particles.SpriteSize=&Multiply_Vector2D_ByFloat(Float=$System.GlobalScale, Vector2D=$StackContext.SpriteSize)) {isSetParameters=true, module=SetVariables_#2}
  ]
  particleUpdateScript [
   &ParticleState()
   &GravityForce(Gravity=&RandomRangeVector(Minimum=<0 0 -900>, Maximum=<0 0 -500>))
   &Drag(Drag=&RandomRangeFloat(Minimum=3, Maximum=5), Use Linear Drag=true, Use Rotational Drag=false)
   &Collision(
    Age Colliding Particles=false
    Module.Control Rest State Via Surface Normal=false
    Module.EnableMaxCollisionCount=false
    Module.Randomize Collision Normal Vector=false
   )
   &CurlNoiseForce(Noise Frequency=15, Noise Quality / Cost=::NewEnumerator0, Noise Strength=500, Pan Noise=false)
   &SolveForcesAndVelocity(Clamp Velocity=false, Limit Acceleration=false)
   &ScaleColor(
    Scale Alpha=&FloatFromCurve(FloatCurve=DataInterface() {propertyValues {Curve {keys [
         {arriveTangent=26.7273, interpMode=RCIM_Cubic, leaveTangent=26.7273, tangentMode=RCTM_User, time=0.1, value=0.0144}
         {interpMode=RCIM_Cubic, tangentMode=RCTM_User, time=0.15, value=1}
         {interpMode=RCIM_Cubic, tangentMode=RCTM_User, time=0.85, value=1}
         {arriveTangent=-1.3701, interpMode=RCIM_Cubic, leaveTangent=-1.3701, tangentMode=RCTM_User, time=1.0011, value=0}
        ]}}})
    Scale RGB=&Lerp_Vector(
     A=<0 0 0>
     B=<12.5 4 0.8>
     Alpha=&FloatFromCurve(FloatCurve=DataInterface() {propertyValues {Curve {keys [
          {arriveTangent=15.3906, interpMode=RCIM_Cubic, leaveTangent=15.3906, tangentMode=RCTM_User, time=-0.0092, value=1}
          {interpMode=RCIM_Cubic, tangentMode=RCTM_User, time=0.0967, value=0.8}
          {interpMode=RCIM_Cubic, tangentMode=RCTM_User, time=0.45, value=0.8}
          <at 0.7 0 RCIM_Cubic>
         ]}}})
    )
    ScaleA=true
    ScaleRGB=true
   )
   &ScaleSpriteSize(Scale Factor=&Vector2DFromFloat(Value=&FloatFromCurve(FloatCurve=DataInterface() {propertyValues {Curve {keys [<at -0.0046 0>,<at 0.025 1>,<at 0.66 1>,<at 1.0011 0.5>]}}}, Scale Curve=2)))
   &ScaleSpriteSizeBySpeed(Max Scale Factor=<0.5 2.2>, Min Scale Factor=<1 0.6>, Velocity Threshold=1500)
  ]
  rendererClasses [&NiagaraSpriteRendererProperties]
  renderers [{rendererClass=&NiagaraSpriteRendererProperties}]
 }
 Streaks_Fat_3 {
  bEnabled=false
  emitterUpdateScript [
   &EmitterState(Life Cycle Mode=::NewEnumerator1, Loop Behavior=::NewEnumerator1, Loop Delay=0.1, Loop Duration=0.4, UseLoopDelay=true)
   &SpawnBurst_Instantaneous(Spawn Count=4, Use Spawn Probability=false, UseLoopCountLimit=false)
   &SpawnBurst_Instantaneous(Spawn Count=2, Spawn Group=1, Use Spawn Probability=false, UseLoopCountLimit=false) {module=SpawnBurst_Instantaneous001}
  ]
  particleSpawnScript [
   &InitializeParticle(
    Color=<rgba 0.3 0.3 0.3 1.5>
    Lifetime Max=1
    Lifetime Min=0.6
    Lifetime Mode=::NewEnumerator1
    Sprite Size=&Multiply_Vector2D_ByFloat(Float=$System.GlobalScale, Vector2D=&UniformRangedVector2D(Minimum=<40 75>, Maximum=<60 130>))
    Sprite Size Mode=::NewEnumerator4
    UsePositionOffset=false
    Write Material Random=true
   )
   &SphereLocation(
    MaskSpawnGroup=false
    Module.Override Randomness=false
    Sphere Radius=&Multiply_Float(A=4, B=$System.GlobalScale)
    Surface Only=true
   )
   &AddVelocityFromPoint(
    Module.Use Falloff Distance=false
    Origin Offset=&Multiply_VectorByFloat(Vector=<0 0 -0.5>, Float=$System.GlobalScale)
    Velocity Strength=&UniformRangedFloat(Minimum=1, Maximum=2.5)
   )
   &SubUVAnimation(End Frame=3, SubUV Animation Mode:ENiagara_SubUVLookupMode=::NewEnumerator2)
   &CameraOffset(Camera Offset Amount=&Multiply_Float(A=&MaskFloatBySpawnGroup(Masked Float=10.01, Passthrough Float=0, Spawn Group Mask=1), B=$System.GlobalScale))
  ]
  particleUpdateScript [
   &ParticleState()
   &Drag(
    Drag=&FloatFromCurve(FloatCurve=DataInterface() {propertyValues {Curve {keys [<at 0.2 1>,<at 0.3 0>]}}}, Scale Curve=150)
    Use Linear Drag=true
    Use Rotational Drag=false
   )
   &GravityForce(Gravity=&Multiply_VectorByFloat(
     Vector=<0 0 -2.2>
     Float=&FloatFromCurve(
      FloatCurve=DataInterface() {propertyValues {Curve {keys [{arriveTangent=3.1357, interpMode=RCIM_Cubic, leaveTangent=3.1357, tangentMode=RCTM_User, time=0.1, value=0}, <at 0.6 1>]}}}
      CurveIndex=$Emitter.Age
     )
    )) {enabled=false}
   &SolveForcesAndVelocity(Clamp Velocity=false, Limit Acceleration=false)
   &ScaleSpriteSize(Scale Factor=&Vector2DFromCurve(
     Vector2Curve=DataInterface() {propertyValues {
       XCurve {keys [
         {arriveTangent=0.4813, interpMode=RCIM_Cubic, leaveTangent=0.4813, tangentMode=RCTM_User, time=0, value=0.625}
         {arriveTangent=1.25, interpMode=RCIM_Cubic, leaveTangent=1.25, time=0.4, value=0.9716}
         {arriveTangent=2.6138, interpMode=RCIM_Cubic, leaveTangent=2.6138, tangentMode=RCTM_User, time=0.9, value=1.75}
        ]}
       YCurve {keys [
         {arriveTangent=9.4573, interpMode=RCIM_Cubic, leaveTangent=9.4573, tangentMode=RCTM_User, time=0, value=0}
         {arriveTangent=0.8193, interpMode=RCIM_Cubic, leaveTangent=0.8193, tangentMode=RCTM_User, time=0.2, value=1.5019}
         <at 1 1.7992 RCIM_Cubic>
        ]}
      }}
     ScaleCurve=<0.8 1.5>
    ))
   &ScaleColor(
    Linear Color Curve=DataInterface() {propertyValues {
      AlphaCurve {keys [
        {arriveTangent=-0.3577, interpMode=RCIM_Cubic, leaveTangent=-0.3577, tangentMode=RCTM_User, time=0.08, value=1}
        {arriveTangent=-1.087, interpMode=RCIM_Cubic, leaveTangent=-1.0164, tangentMode=RCTM_Break, time=0.3702, value=1}
        <at 1 0>
       ]}
      BlueCurve {keys [<at 0.1756 0.4583>,<at 0.3677 0.2>,<at 0.6491 0.3>,<at 1 0.45>]}
      GreenCurve {keys [<at 0.1756 0.4464>,<at 0.3677 0.2>,<at 0.6491 0.3>,<at 1 0.45>]}
      RedCurve {keys [<at 0.1756 0.4471>,<at 0.3677 0.2>,<at 0.6491 0.3>,<at 1 0.45>]}
     }}
    Scale Mode=::NewEnumerator2
   )
   &DynamicMaterialParameters(
    Index 0 Param 1=&Multiply_Float(A=100.01, B=$System.GlobalScale)
    Index 0 Param 4=&FloatFromCurve(FloatCurve=DataInterface() {propertyValues {Curve {keys [{interpMode=RCIM_Cubic, tangentMode=RCTM_User, time=0, value=0}, {interpMode=RCIM_Cubic, tangentMode=RCTM_User, time=0.4, value=1}]}}})
    Index 1 Param 1=&FloatFromCurve(
     FloatCurve=DataInterface() {propertyValues {Curve {keys [
         {interpMode=RCIM_Cubic, tangentMode=RCTM_User, time=0, value=0}
         {arriveTangent=0.6295, interpMode=RCIM_Cubic, leaveTangent=0.6295, tangentMode=RCTM_User, time=1, value=1}
        ]}}}
     Scale Curve=0.42
    )
    Index 1 Param 2=&FloatFromCurve(
     FloatCurve=DataInterface() {propertyValues {Curve {keys [
         {arriveTangent=-100.8588, interpMode=RCIM_Cubic, leaveTangent=-100.8588, tangentMode=RCTM_User, time=0, value=30}
         {arriveTangent=-26.7969, interpMode=RCIM_Cubic, leaveTangent=-26.7969, tangentMode=RCTM_User, time=0.15, value=4.3564}
         {arriveTangent=-2.6967, interpMode=RCIM_Cubic, leaveTangent=-2.6967, tangentMode=RCTM_User, time=0.3, value=0}
        ]}}}
     CurveIndex=$Emitter.Age
     Scale Curve=&MaskFloatBySpawnGroup(Masked Float=0, Passthrough Float=5, Spawn Group Mask=1)
    )
    Index 1 Param 3=&FloatFromCurve(
     FloatCurve=DataInterface() {propertyValues {Curve {keys [{interpMode=RCIM_Cubic, tangentMode=RCTM_User, time=0.2, value=0.5}, {interpMode=RCIM_Cubic, tangentMode=RCTM_User, time=1, value=1}]}}}
     Scale Curve=2
    )
    Index 1 Param 4=&FloatFromCurve(FloatCurve=DataInterface() {propertyValues {Curve {keys [<at 0 1>,<at 1 0>]}}}, Scale Curve=-0.3)
    Module.Param0WriteEnabled=true
    Module.Param0WriteEnabled1=true
    Module.Param1WriteEnabled=false
    Module.Param1WriteEnabled1=true
    Module.Param2WriteEnabled=false
    Module.Param2WriteEnabled1=true
    Module.Param3WriteEnabled=true
    Module.Param3WriteEnabled1=true
    Write Parameter Index 1=true
   )
  ]
  rendererClasses [&NiagaraSpriteRendererProperties]
  renderers [{rendererClass=&NiagaraSpriteRendererProperties}]
 }
}
summary {
 systemName=NS_Emote_PowerFarmer_ExplosionSmall
 emitters [
  {
   emitterName=Smoke_Trail_Leader
   simTarget=GPUComputeSim
   rendererClasses [&NiagaraSpriteRendererProperties, &NiagaraMeshRendererProperties]
  }
  {
   bEnabled=false
   emitterName=GlowSphereMesh
   rendererClasses [&NiagaraMeshRendererProperties, &NiagaraSpriteRendererProperties]
  }
  {
   bEnabled=false
   emitterName=Intro_Flash
   rendererClasses [&NiagaraMeshRendererProperties, &NiagaraComponentRendererProperties]
  }
  {
   bEnabled=false
   emitterName=Smoke_Trail_Leader001
   rendererClasses [&NiagaraSpriteRendererProperties, &NiagaraMeshRendererProperties]
  }
  {bEnabled=false, emitterName=Ribbon001, rendererClasses [&NiagaraRibbonRendererProperties]}
  {bEnabled=false, emitterName=Flash_wide, rendererClasses [&NiagaraMeshRendererProperties]}
  {bEnabled=false, emitterName=Sparkles001, simTarget=GPUComputeSim, rendererClasses [&NiagaraSpriteRendererProperties]}
  {bEnabled=false, emitterName=Blast_Stylized, rendererClasses [&NiagaraSpriteRendererProperties]}
  {emitterName=Light, rendererClasses [&NiagaraLightRendererProperties]}
  {bEnabled=false, emitterName=NE_Explosion_Main, rendererClasses [&NiagaraMeshRendererProperties]}
  {bEnabled=false, emitterName=Streaks_Fat_3, rendererClasses [&NiagaraSpriteRendererProperties]}
  {emitterName=Dark_Chunks_6, simTarget=GPUComputeSim, rendererClasses [&NiagaraSpriteRendererProperties]}
  {emitterName=Sparkles002, rendererClasses [&NiagaraSpriteRendererProperties]}
  {emitterName=Smoke_Ground001, rendererClasses [&NiagaraSpriteRendererProperties]}
  {emitterName=Smoke_Ground, rendererClasses [&NiagaraSpriteRendererProperties]}
  {emitterName=Dark_Chunks001, simTarget=GPUComputeSim, rendererClasses [&NiagaraSpriteRendererProperties]}
  {emitterName=Sparkles, rendererClasses [&NiagaraSpriteRendererProperties]}
 ]
 userVariables [
  User.BlastRadiusScalar {type=NiagaraFloat, defaultValue=1, description=""}
  User.Global Color {type=LinearColor, description="", defaultValue {struct=LinearColor, value=<rgba 4 0.5231 0 1>}}
  User.Global Scale {type=NiagaraFloat, defaultValue=0.75, description=""}
  User.Grid2DScale {type=NiagaraFloat, defaultValue=1.1, description=""}
  User.ResolutionMaxAxis {type=NiagaraInt32}
  User.ShowOverlays {type=NiagaraBool}
  User.WorldSpaceSize {type=Vector3f, description="", defaultValue {struct=Vector3f, value=<0 0 0>}}
 ]
}
